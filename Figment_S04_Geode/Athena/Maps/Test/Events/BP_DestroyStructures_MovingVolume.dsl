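; Test-map event actor that smashes structures in the path of the moving actor it is attached to.
; BeginDestructionChecks starts a looping timer for DamageStructuresInPath, and EndDestructionChecks clears it.
; Each tick of the timer sphere-traces from its location toward a point predicted from the parent's velocity, ignoring that parent.
; Hit actors that can be damaged receive global environment damage of a configurable amount.
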(event EventBeginPlay)

(event Custom|BeginDestructionChecks
  ;[1 1 1] Perform the damage sweep periodically.
  (Utilities|Time|SetTimerbyEvent (EventDispatchers|CreateEvent :!event Custom|DamageStructuresInPath) (Variables|Default|GetUpdateIntervalForDamageVolume) true))

(event Custom|DamageStructuresInPath
  (bind _returnvalue (Actor|GetAttachParentActor))
  ;[0.3591 1 0.5264] Damage structures in an AOE
  (Utilities|IsValid _returnvalue
    (:Is%20Valid
      (bind (_outhits _returnvalue_2) (Collision|MultiSphereTraceForObjects (Transformation|GetActorLocation) (+ (Transformation|GetActorLocation) (* (Transformation|GetVelocity _returnvalue) (* (Utilities|Time|GetWorldDeltaSeconds) (Variables|Default|GetSmashPathPredictionMultiplier)))) (Variables|Default|GetDestructionSphereRadius) (Utilities|Array|MakeArray "ObjectTypeQuery1") false (Utilities|Array|MakeArray (Actor|GetAttachParentActor)) "None" true "(R=1.000000,G=0.000000,B=0.000000,A=1.000000)" "(R=0.000000,G=1.000000,B=0.000000,A=1.000000)" 15.0))
      (if _returnvalue_2
        (for _array_element _outhits
          (if (Variables|/Script/Engine.Actor|GetbCanBeDamaged (.HitActor (Collision|BreakHitResult _array_element)))
            (Environment|ApplyGlobalEnvironmentDamagetoActor (.HitActor (Collision|BreakHitResult _array_element)) (Variables|Default|GetAmount_EnvironmentDamage))))))))

(event Custom|EndDestructionChecks
  ;[1 1 1] Stop checking for damage.
  (Utilities|Time|ClearTimerbyFunctionName 0 "DamageStructuresInPath"))

(fn ConstructionScript)
