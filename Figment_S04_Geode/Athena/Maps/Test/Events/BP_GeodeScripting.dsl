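; Geode event Blueprint that runs the live rocket launch countdown and Level Sequence, with debug test events.
; Server watches calendar events to set countdown and sequence start times, time of day, thruster prop and god ray timers.
; Replicated countdown drives a client timer updating material-collection digits, klaxons and vapor FX.
; Server records playback position each tick; OnRep resyncs client sequence players and hides the original rocket during replays.
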(event Custom|Event_ThrusterImpact
  ;[1 1 1] Make the Thruster prop appear at Anarchy
  (Class|BPAthenaCalendarEventPropHandler|ActivateEarly (Variables|Default|GetPropHandler_Thruster)))

(event EventTick
  (params (DeltaSeconds :type Float%20%28single-precision%29))
  ;[1 1 1] Tick: Keep history of playback position
  (Networking|SwitchHasAuthority
    (:Authority
      (Variables|Default|SetSequenceProgress (CallFunction|GetLocalSequencePlayerProgress)))))

(event EventBeginPlay
  ;[1 1 1] Determine countdown time.
  (Utilities|FlowControl|Sequence
    (:then_0
      (bind _asfort_game_mode_zone (Utilities|Casting|CastToFortGameModeZone (Game|GetGameMode)))
      (EventDispatchers|BindEventtoOnHandleMatchHasStarted _asfort_game_mode_zone (EventDispatchers|CreateEvent :!event Custom|InitializeLevel)))))

(event Custom|TOD
  (params (TimeOfDay :type Float%20%28double-precision%29))
  ;[1 0 0] Debug
  (Networking|SwitchHasAuthority
    (:Authority
      (TimeOfDay|SetTimeOfDay TimeOfDay :!class /Script/FortniteGame.FortKismetLibrary))))

(event Custom|InitializeLevel
  ;[1 1 1] Determine countdown time.
  (Variables|Default|SetShouldPlayCrackAudio false)
  (Networking|SwitchHasAuthority
    (:Authority
      (Utilities|WatchShouldRunEventinPlaylist
        (:ShouldRun
          (Utilities|FlowControl|Sequence
            (:then_0
              (Variables|Default|SetShowPhase2 0.0))
            (:then_1
              ;;[1 1 1] Countdown
              (bind (_eventname _timeuntilend) (MCPContext|WatchCalendarEvent (Variables|Default|GetEvent_Countdown))
                (:OnEventActive
                  (Variables|Default|SetTimeUntilCountdownEnd _timeuntilend)
                  (CallFunction|ContinueCountdownEvent))
                (:OnEventUpdated
                  (Variables|Default|SetTimeUntilCountdownEnd _timeuntilend))))
            (:then_2
              ;;[1 1 1] Launch Event
              (bind (_eventname_2 _timeuntilend_2) (MCPContext|WatchCalendarEvent (Variables|Default|GetEvent_Begin))
                (:OnEventActive
                  (Variables|Default|SetSequenceStartTime (- (Variables|Default|GetTotalLaunchSequenceLength) (Math|Timespan|GetTotalSeconds _timeuntilend_2)))
                  (Utilities|FlowControl|Sequence
                    (:then_0
                      (Utilities|FlowControl|Delay 2.0)
                      (CallFunction|ContinueLaunch))
                    (:then_1
                      (Utilities|FlowControl|Delay (Variables|Default|GetDurationLoadingFudgeFactor))
                      (if (< (Math|Timespan|GetTotalSeconds _timeuntilend_2) (Variables|Default|GetEndTimeMinusThrusterImpactTime))
                        (CallFunction|EventThrusterImpact)))))))
            (:then_3
              (MCPContext|WatchCalendarEvent (Variables|Default|GetEvent_Over)
                (:OnEventActive
                  (Variables|Default|SetVillainBase 0.0))))))))))

(event Custom|TestCountdown
  (params (Seconds :type Integer) (LaunchStartTime :type Float%20%28double-precision%29))
  ;[1 0 0] Debug
  (Networking|SwitchHasAuthority
    (:Authority
      (Variables|Default|SetTimeUntilCountdownEnd (Math|Timespan|MakeTimespan 0 0 0 Seconds))
      (CallFunction|ContinueCountdownEvent)
      (if (not (== LaunchStartTime 0.0))
        (Utilities|Time|SetTimerbyEvent (delegate Custom|AlsoTestLaunch) (Math|Conversions|ToFloat(Integer) (- Seconds 60)))))))

(event Custom|ContinueCountdownEvent
  ;[1 1 1] Determine countdown time.
  ;;[1 1 1] Countdown
  (Variables|Default|SetServerWorldSecondsAtLaunch (+ (GameState|GetServerWorldTimeSeconds (Game|GetGameState)) (Math|Timespan|GetTotalSeconds (Variables|Default|GetTimeUntilCountdownEnd))))
  (Variables|Default|SetShowPhase2 1.0))

(event Custom|AlsoTestLaunch
  ;[1 0 0] Debug
  (Variables|Default|SetSequenceStartTime (stale Custom|TestCountdown LaunchStartTime))
  (CallFunction|ContinueLaunch))

(event Custom|ContinueLaunch
  ;[1 1 1] Determine countdown time.
  (CallFunction|LaunchSequence)
  (TimeOfDay|SetTimeOfDay (CallFunction|TimeOfDayfromSequencePosition (- (Variables|Default|GetSequenceStartTime) 6e+01)) :!class /Script/FortniteGame.FortKismetLibrary))

(event Custom|TestLaunch
  (params (Seconds :type Float%20%28double-precision%29))
  ;[1 0 0] Debug
  (Networking|SwitchHasAuthority
    (:Authority
      (Variables|Default|SetSequenceStartTime Seconds)
      (CallFunction|ContinueLaunch))))

(event Custom|CountdownTimer_New
  (bind _timespanremaining (Variables|Default|GetTimespanRemaining))
  (bind _returnvalue (Math|Timespan|GetSeconds _timespanremaining))
  (bind _returnvalue_2 (Utilities|String|ToString(Integer) _returnvalue))
  (bind _returnvalue_3 (Math|Timespan|GetHours _timespanremaining))
  (bind _returnvalue_4 (Utilities|String|ToString(Integer) _returnvalue_3))
  (bind _returnvalue_5 (Math|Timespan|GetDays _timespanremaining))
  (bind _returnvalue_6 (Utilities|String|ToString(Integer) _returnvalue_5))
  (bind _returnvalue_7 (Utilities|String|Reverse _returnvalue_6))
  (bind _returnvalue_8 (Math|Timespan|GetMinutes _timespanremaining))
  (bind _returnvalue_9 (Utilities|String|ToString(Integer) _returnvalue_8))
  (bind _returnvalue_10 (Utilities|String|Reverse _returnvalue_9))
  (bind _returnvalue_11 (Utilities|String|Reverse _returnvalue_2))
  (bind _returnvalue_12 (Utilities|String|Reverse _returnvalue_4))
  (bind _secondsremaining (Variables|Default|GetSecondsRemaining))
  (bind _ps_countdownvaporfx (Variables|Default|GetPS_CountdownVaporFX))
  ;[1 1 1] Update Countdown every second.
  (bind _output_get (Variables|Default|SetSecondsRemaining (Math|Float|Truncate (- (Variables|Default|GetServerWorldSecondsAtLaunch) (GameState|GetServerWorldTimeSeconds (Game|GetGameState))))))
  (Default|CallUpdateCountdownTimer _output_get)
  ;; UpdateCountdownVisuals
  (Utilities|FlowControl|Sequence
    (:then_0
      (Rendering|Material|SetScalarParameterValue "/Game/Athena/Events/Generic/CountdownTimer/Materials/MPC_Countdown.MPC_Countdown" "UseStatic" :!class /Script/Engine.KismetMaterialLibrary)
      (if (< _secondsremaining 0)
        (Rendering|Material|SetScalarParameterValue "/Game/Athena/Events/Generic/CountdownTimer/Materials/MPC_Countdown.MPC_Countdown" "Launch" 1.0 :!class /Script/Engine.KismetMaterialLibrary)
        (else
          (Variables|Default|SetTimespanRemaining (Math|Timespan|MakeTimespan 0 0 0 _secondsremaining))
          (Rendering|Material|SetScalarParameterValue "/Game/Athena/Events/Generic/CountdownTimer/Materials/MPC_Countdown.MPC_Countdown" "Launch" :!class /Script/Engine.KismetMaterialLibrary)
          (Rendering|Material|SetScalarParameterValue "/Game/Athena/Events/Generic/CountdownTimer/Materials/MPC_Countdown.MPC_Countdown" "Digit1" (Utilities|String|StringToFloat (Utilities|String|GetSubstring _returnvalue_11)) :!class /Script/Engine.KismetMaterialLibrary)
          (Rendering|Material|SetScalarParameterValue "/Game/Athena/Events/Generic/CountdownTimer/Materials/MPC_Countdown.MPC_Countdown" "Digit2" (Utilities|String|StringToFloat (Utilities|String|GetSubstring _returnvalue_11 1)) :!class /Script/Engine.KismetMaterialLibrary)
          (Rendering|Material|SetScalarParameterValue "/Game/Athena/Events/Generic/CountdownTimer/Materials/MPC_Countdown.MPC_Countdown" "Digit3" (Utilities|String|StringToFloat (Utilities|String|GetSubstring _returnvalue_10)) :!class /Script/Engine.KismetMaterialLibrary)
          (Rendering|Material|SetScalarParameterValue "/Game/Athena/Events/Generic/CountdownTimer/Materials/MPC_Countdown.MPC_Countdown" "Digit4" (Utilities|String|StringToFloat (Utilities|String|GetSubstring _returnvalue_10 1)) :!class /Script/Engine.KismetMaterialLibrary)
          (Rendering|Material|SetScalarParameterValue "/Game/Athena/Events/Generic/CountdownTimer/Materials/MPC_Countdown.MPC_Countdown" "Digit5" (Utilities|String|StringToFloat (Utilities|String|GetSubstring _returnvalue_12)) :!class /Script/Engine.KismetMaterialLibrary)
          (Rendering|Material|SetScalarParameterValue "/Game/Athena/Events/Generic/CountdownTimer/Materials/MPC_Countdown.MPC_Countdown" "Digit6" (Utilities|String|StringToFloat (Utilities|String|GetSubstring _returnvalue_12 1)) :!class /Script/Engine.KismetMaterialLibrary)
          (Rendering|Material|SetScalarParameterValue "/Game/Athena/Events/Generic/CountdownTimer/Materials/MPC_Countdown.MPC_Countdown" "Digit7" (Utilities|String|StringToFloat (Utilities|String|GetSubstring _returnvalue_7)) :!class /Script/Engine.KismetMaterialLibrary)
          (Rendering|Material|SetScalarParameterValue "/Game/Athena/Events/Generic/CountdownTimer/Materials/MPC_Countdown.MPC_Countdown" "Digit8" (Utilities|String|StringToFloat (Utilities|String|GetSubstring _returnvalue_7 1)) :!class /Script/Engine.KismetMaterialLibrary)
          (Rendering|Material|SetScalarParameterValue "/Game/Athena/Events/Generic/CountdownTimer/Materials/MPC_Countdown.MPC_Countdown" "ProgressNormalized" (Math|Float|Clamp(Float) (/ (Math|Timespan|GetTotalSeconds _timespanremaining) (Variables|Default|GetCountdownTotalSeconds))) :!class /Script/Engine.KismetMaterialLibrary))))
    (:then_1
      _))
  (CallFunction|HandleExtraKlaxonAudio (Variables|Default|GetSecondsRemaining))
  (if (<= (Variables|Default|GetSecondsRemaining) (Utilities|Array|Get(acopy) (Variables|Default|GetEventSeconds) (Variables|Default|GetEventIndex)))
    (Utilities|FlowControl|Sequence
      (:then_0
        ;;[1 1 1] Special event times.
        (switch string (Utilities|String|ToString(Integer) (Utilities|Array|Get(aref) (Variables|Default|GetEventSeconds) (Variables|Default|GetEventIndex)))
          (:Default
            (if (and (== (Math|Integer|%25%28Integer%29 (Variables|Default|GetSecondsRemaining) 3600) 0) (> (Variables|Default|GetSecondsRemaining) 0))
              (Development|PrintString "Hourly alarm")
              (Audio|Components|Audio|SetSound (Variables|/Script/Engine.AmbientSound|GetAudioComponent (Variables|Default|GetKlaxon3_Cue)) (Variables|Default|GetAudio_KlaxonLoop3))
              (CallFunction|EventPlayKlaxonCue)))
          (:600
            _)
          (:540
            _)
          (:480
            _)
          (:420
            _)
          (:360
            _)
          (:300
            _)
          (:240
            _)
          (:180
            _)
          (:120
            _)
          (:-1
            (Utilities|Time|ClearandInvalidateTimerbyHandle (Variables|Default|GetCountdownTimerHandle))
            (bind _audio_steamloop (Variables|Default|GetAudio_SteamLoop)
              (:then
                (Audio|Components|Audio|FadeOut _audio_steamloop 1.0 0.0 "Linear" :!class /Script/Engine.AudioComponent))
              (:else))))
        (CallFunction|EventPlayKlaxonCue)
        (Utilities|FlowControl|DoOnce false
          (:Completed
            (Components|Activation|Activate (Variables|/Script/CascadeRuntime.Emitter|GetParticleSystemComponent _ps_countdownvaporfx) false)
            (bind _returnvalue_13 (Audio|SpawnSoundatLocation "/Figment_S04_Geode/Athena/Sounds/Events/Geode/Geode_SteamBuildup_Loop_Cue.Geode_SteamBuildup_Loop_Cue" (Transformation|GetActorLocation _ps_countdownvaporfx)))
            (bind _output_get_2 (Variables|Default|SetAudio_SteamLoop _returnvalue_13))
            ;[1 1 1] Within Final 10 minutes. Start Vapor FX.
            (Audio|Components|Audio|FadeIn _output_get_2 3.0 1.0 0.0 "Linear" :!class /Script/Engine.AudioComponent))))
      (:then_1
        (Math|Integer|IncrementInt (Variables|Default|GetEventIndex))))
    (else
      _)))

(event Custom|Event_PlayKlaxonCue
  ;[1 1 1] Update Countdown every second.
  ;;[1 1 1] Play Klaxon Sound Cue
  (Audio|Components|Audio|Play (Variables|/Script/Engine.AmbientSound|GetAudioComponent (Variables|Default|GetKlaxon3_Cue)) 0.0))

(event Custom|DisableHologram
  ;[1 1 1] Update Countdown every second.
  (Default|CallCountdownFinished30s)
  (Rendering|Material|SetScalarParameterValue "/Game/Athena/Events/Generic/CountdownTimer/Materials/MPC_Countdown.MPC_Countdown" "Launch" 1.0 :!class /Script/Engine.KismetMaterialLibrary)
  (Variables|Default|SetShowPhase2 0.0)
  (Variables|Default|SetVillainBase 0.0))

(event Custom|DoLaunch
  ;[1 1 1] Console Event to begin sequence. Use 'cheat ce ConsoleEvent_BeginLaunch'
  (Networking|SwitchHasAuthority
    (:Authority
      (Development|PrintString "ServerLaunchEvent" true true "(R=1.000000,G=0.000000,B=0.064100,A=1.000000)" 1e+01)
      (CallFunction|LaunchSequence))))

(event Custom|LaunchSequence
  (bind _launchrocketsequencesoft (Variables|Default|GetLaunchRocketSequenceSoft))
  (bind _output (Utilities|ResolveSoftReference _launchrocketsequencesoft))
  (bind _sequenceplayer (Variables|/Script/LevelSequence.LevelSequenceActor|GetSequencePlayer _output))
  ;[1 1 1] Start Launch Sequence from the beginning.
  (Networking|SwitchHasAuthority
    (:Authority
      (Sequencer|Player|SetPlaybackPosition _sequenceplayer (Utilities|Struct|MakeMovieSceneSequencePlaybackParams "(FrameNumber=(Value=0),SubFrame=0.000000)" (Variables|Default|GetSequenceStartTime) "" "(Hours=0,Minutes=0,Seconds=0,Frames=0,Subframe=0.000000,bDropFrameFormat=False)" "Time" "Jump"))
      (Development|PrintString "Launch rocket!")
      (Sequencer|Player|Play _sequenceplayer :!class /Script/MovieScene.MovieSceneSequencePlayer))))

(event Custom|Event_TurnOnGodRay
  ;[1 1 1] Event Turn on/off God Ray.
  (Development|PrintString "Turn on God Ray")
  (Variables|/Figment_S04_Geode/Athena/Effects/Meshes/Geode/BP_GeodeGodRay.BP_GeodeGodRay_C|SetIsRayVisible 1 (Variables|Default|GetGodRayLair)))

(event Custom|Event_TurnOffGodRay
  ;[1 1 1] Event Turn on/off God Ray.
  (Development|PrintString "Turn OFF God Ray.")
  (Variables|/Figment_S04_Geode/Athena/Effects/Meshes/Geode/BP_GeodeGodRay.BP_GeodeGodRay_C|SetIsRayVisible 0 (Variables|Default|GetGodRayLair)))

(fn GetLocalSequencePlayerProgress
  (bind _launchrocketsequencesoft (Variables|Default|GetLaunchRocketSequenceSoft))
  (bind _output (Utilities|ResolveSoftReference _launchrocketsequencesoft))
  (bind _sequenceplayer (Variables|/Script/LevelSequence.LevelSequenceActor|GetSequencePlayer _output))
  (Utilities|IsValid _output
    (:Is%20Valid
      (Utilities|IsValid _sequenceplayer
        (:Is%20Valid
          (return (Utilities|TimeManagement|QualifiedFrameTimeToSeconds (Sequencer|Player|GetCurrentTime _sequenceplayer))))
        (:Is%20Not%20Valid
          _)))
    (:Is%20Not%20Valid
      _))
  (return 0.0))

(fn HandleExtraKlaxonAudio
  (params (Seconds :type Integer))
  (bind _klaxon3_cue (Variables|Default|GetKlaxon3_Cue))
  (bind _audiocomponent (Variables|/Script/Engine.AmbientSound|GetAudioComponent _klaxon3_cue))
  (if (Math|Boolean|ANDBoolean (<= Seconds 60) (== (Math|Integer|%25%28Integer%29 Seconds 10) 0) (> Seconds 0))
    (Audio|Components|Audio|SetSound _audiocomponent (Variables|Default|GetAudio_KlaxonLooping))
    (Audio|Components|Audio|Play _audiocomponent 0.0)))

(fn HideOriginalRocketDuringReplay
  (bind _originalrocketprop (Variables|Default|GetOriginalRocketProp))
  (bind _output (Utilities|ResolveSoftReference _originalrocketprop))
  ; This is a workaround to hide the rocket during the replay of the rocket launch
  (if (Replay|GetIsPlayingReplay)
    (if (!= (Variables|Default|GetSequenceProgress) 0)
      (Utilities|IsValid _output
        (:Is%20Valid
          (if (not (Variables|/Script/Engine.Actor|GetbHidden _output))
            (Rendering|SetActorHiddenInGame _output true)
            (Collision|SetCollisionEnabled (Variables|/Figment_S04/Environments/Sets/Super_V_Lair/Props/BP/BP_SVLair_GeodeObject2.BP_SVLair_GeodeObject2_C|GetMissileSilo_Stage2_Lower _output) "NoCollision")))))))

(fn OnRep_SequenceProgress
  (bind _sequencestarttime (Variables|Default|GetSequenceStartTime))
  (bind _sequenceprogress (Variables|Default|GetSequenceProgress))
  (bind _returnvalue (== _sequenceprogress _sequencestarttime))
  (bind _launchrocketsequencesoft (Variables|Default|GetLaunchRocketSequenceSoft))
  (bind _output (Utilities|ResolveSoftReference _launchrocketsequencesoft))
  (bind _sequenceplayer (Variables|/Script/LevelSequence.LevelSequenceActor|GetSequencePlayer _output))
  (Networking|SwitchHasAuthority
    (:Remote
      (Utilities|FlowControl|Sequence
        (:then_0
          (Utilities|IsValid (Utilities|ResolveSoftReference (Variables|Default|GetLaunchRocketSequenceSoft))
            (:Is%20Valid
              (if (not (Math|Float|NearlyEqual(Float) (CallFunction|GetLocalSequencePlayerProgress) (Variables|Default|GetSequenceProgress) 0.1))
                (Utilities|FlowControl|Sequence
                  (:then_0
                    ;[1 1 1] Should we Play?
                    (if (Sequencer|Player|IsPlaying _sequenceplayer :!class /Script/MovieScene.MovieSceneSequencePlayer)
                      (if _returnvalue
                        (Sequencer|Player|Pause _sequenceplayer))
                      (elif _returnvalue
                        _
                        (else
                          (Sequencer|Player|Play _sequenceplayer :!class /Script/MovieScene.MovieSceneSequencePlayer)))))
                  (:then_1
                    ;[1 1 1] Set Position
                    (Sequencer|Player|SetPlaybackPosition (Variables|/Script/LevelSequence.LevelSequenceActor|GetSequencePlayer (Utilities|ResolveSoftReference (Variables|Default|GetLaunchRocketSequenceSoft))) (Utilities|Struct|MakeMovieSceneSequencePlaybackParams "(FrameNumber=(Value=0),SubFrame=0.000000)" (Variables|Default|GetSequenceProgress) "" "(Hours=0,Minutes=0,Seconds=0,Frames=0,Subframe=0.000000,bDropFrameFormat=False)" "Time" (select (Replay|GetIsPlayingReplay) "Jump" (select (> (CallFunction|GetLocalSequencePlayerProgress) (Variables|Default|GetSequenceProgress)) "Jump" "Play"))))))))))
        (:then_1
          (CallFunction|HideOriginalRocketDuringReplay))))))

(fn OnRep_ServerWorldSecondsUntilLaunch
  (bind _timeuntilcountdownend (Variables|Default|GetTimeUntilCountdownEnd))
  (bind _returnvalue (Math|Timespan|GetTotalMinutes _timeuntilcountdownend))
  (bind _returnvalue_2 (Math|Timespan|GetTotalSeconds _timeuntilcountdownend))
  (Networking|SwitchHasAuthority
    (:Authority
      (Utilities|FlowControl|Sequence
        (:then_0
          ;[1 1 1] Turn on/off God Ray
          (if (<= _returnvalue 1.2e+02)
            (Utilities|Time|SetTimerbyEvent (EventDispatchers|CreateEvent :!event Event_TurnOnGodRay) (Math|Float|Clamp(Float) (* (- _returnvalue 6e+01) 6e+01) 0.1 1e+04))
            (Utilities|Time|SetTimerbyEvent (EventDispatchers|CreateEvent :!event Event_TurnOffGodRay) (+ (Math|Timespan|GetTotalSeconds _timeuntilcountdownend) (Variables|Default|GetPostLaunchGodRayDuration)))))
        (:then_1
          (if (<= _returnvalue 33.0)
            (CallFunction|SetProperTimeOfDay _returnvalue_2)
            (Utilities|Time|SetTimerbyEvent (EventDispatchers|CreateEvent :!event DisableHologram) (+ _returnvalue_2 29.0))))))
    (:Remote
      (bind _output_get (Variables|Default|SetSecondsRemaining (Math|Float|Truncate (- (Variables|Default|GetServerWorldSecondsAtLaunch) (GameState|GetServerWorldTimeSeconds (Game|GetGameState))))))
      (Utilities|FlowControl|Sequence
        (:then_0
          (bind _returnvalue_3 (Utilities|Time|SetTimerbyEvent (EventDispatchers|CreateEvent :!event CountdownTimer_New) 1.0 true))
          (Variables|Default|SetCountdownTimerHandle _returnvalue_3)
          (Audio|Components|Audio|Stop (Variables|/Script/Engine.AmbientSound|GetAudioComponent (Variables|Default|GetKlaxon3_Cue)))
          (Variables|Default|SetEventIndex 0)
          (CallFunction|CountdownTimerNew)
          (bind _array_element (Utilities|Array|ForEachLoopwithBreak (Variables|Default|GetEventSeconds))
            (:LoopBody
              (if (> (Variables|Default|GetSecondsRemaining) _array_element)
                _)
              (Math|Integer|IncrementInt (Variables|Default|GetEventIndex)))))
        (:then_1
          (if (<= _output_get 3700)
            (Rendering|Material|SetScalarParameterValue "/Game/Athena/SafeZone/MPC_SafeZoneClouds.MPC_SafeZoneClouds" "PreventStormShrinking" 1.0 :!class /Script/Engine.KismetMaterialLibrary)))
        (:then_2
          ;[1 1 1] Countdown active. Unhide Klaxon FX.
          (Class|BPGeodeBuildUpFX|EventUnhideKlaxonFX (Variables|Default|GetBP_CountdownFX)))))))

(fn OnRep_ShouldPlayCrackAudio
  (if (Variables|Default|GetShouldPlayCrackAudio)
    (Audio|SpawnSoundatLocation "/Figment_S04_Geode/Athena/Sounds/Events/Geode/Geode_Tear_InWorldLoop_Cue.Geode_Tear_InWorldLoop_Cue" "5769.480957,5569.282227,106302.406250")))

(fn OnRep_ShowPhase2
  (Rendering|Material|SetScalarParameterValue "/Game/Athena/Events/Generic/CountdownTimer/Materials/MPC_Countdown.MPC_Countdown" "ShowPhase2" (Variables|Default|GetShowPhase2) :!class /Script/Engine.KismetMaterialLibrary))

(fn OnRep_VillainBase
  (Rendering|Material|SetScalarParameterValue "/Game/Athena/Events/Generic/CountdownTimer/Materials/MPC_Countdown.MPC_Countdown" "VillainBase" (Variables|Default|GetVillainBase) :!class /Script/Engine.KismetMaterialLibrary))

(fn SequencerShowCrack
  (bind _outactors (Actor|GetAllActorsOfClass "/Figment_S04/Temporary/Athena/Events/BP_Athena_Event_Components.BP_Athena_Event_Components_C"))
  (for _array_element _outactors
    (Class|BPAthenaEventComponents|EnablePostSequenceCrack _array_element)))

(fn SetProperTimeOfDay
  (params (Seconds :type Float%20%28double-precision%29))
  (bind _returnvalue (/ Seconds 6e+01))
  (bind _returnvalue_2 (* _returnvalue 1.333333))
  (bind _returnvalue_3 (< _returnvalue_2 0.0))
  (bind _returnvalue_4 (+ 24.0 _returnvalue_2))
  (bind _returnvalue_5 (select _returnvalue_3 _returnvalue_4 _returnvalue_2))
  (bind _returnvalue_6 (Math|Float|%25%28Float%29 _returnvalue_5 24.0))
  (bind _launchtimeofday (Variables|Default|GetLaunchTimeOfDay))
  (bind _returnvalue_7 (- _launchtimeofday _returnvalue_6))
  (bind _returnvalue_8 (< _returnvalue_7 0.0))
  (bind _returnvalue_9 (+ 24.0 _returnvalue_7))
  (bind _returnvalue_10 (select _returnvalue_8 _returnvalue_9 _returnvalue_7))
  (LocalVariables|SetTime _returnvalue_10)
  (TimeOfDay|SetTimeOfDay _returnvalue_10 :!class /Script/FortniteGame.FortKismetLibrary)
  (Development|PrintString (Utilities|String|Append "Initial start ToD:" (Utilities|String|ToString(Float) _returnvalue_10))))

(fn TimeOfDayFromSequencePosition
  (params (A :type Float%20%28double-precision%29))
  (return (select (< (Math|Float|%25%28Float%29 (+ (Variables|Default|GetLaunchTimeOfDay) (* (/ A 6e+01) 1.333333)) 24.0) 0.0) (+ (Math|Float|%25%28Float%29 (+ (Variables|Default|GetLaunchTimeOfDay) (* (/ A 6e+01) 1.333333)) 24.0) 24.0) (Math|Float|%25%28Float%29 (+ (Variables|Default|GetLaunchTimeOfDay) (* (/ A 6e+01) 1.333333)) 24.0))))

(fn ConstructionScript)
