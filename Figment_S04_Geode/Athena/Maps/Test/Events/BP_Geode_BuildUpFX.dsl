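; BP_Geode_BuildUpFX is a Blueprint actor for a Geode test-map event that triggers klaxon build-up effects.
; Its custom event Event_UnhideKlaxonFX loops over all child actors, casts each to BP_Geode_Klaxon and calls EventStartFlash on it.
; BeginPlay, overlap, Tick and the construction script are empty, so the effect runs only when that event is called.
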(event EventBeginPlay)

(event Collision|EventActorBeginOverlap
  (params (OtherActor :type Actor%20Object%20Reference)))

(event EventTick
  (params (DeltaSeconds :type Float%20%28single-precision%29)))

(event Custom|Event_UnhideKlaxonFX
  (for _array_element (Actor|GetAllChildActors true)
    (bind _asbp_geode_klaxon (Utilities|Casting|CastToBP_Geode_Klaxon _array_element))
    (Class|BPGeodeKlaxon|EventStartFlash _asbp_geode_klaxon)))

(fn ConstructionScript)
