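; BP_Geode_Klaxon is a Blueprint for an emergency klaxon light in a Geode test event map that switches to a flashing state on demand.
; The custom event Event_StartFlash swaps the light mesh's first material to KlaxonLit, then activates the NS_Geode_Lights_Loop effect component with a reset.
; No stop event or caller appears here, and the construction script is empty.
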(event Custom|Event_StartFlash
  (Rendering|Material|SetMaterial (Variables|Default|GetS_Mission_Missle_EmergencyLight) 0 (Variables|Default|GetKlaxonLit))
  (Components|Activation|Activate (Variables|Default|GetNS_Geode_Lights_Loop) true))

(fn ConstructionScript)
