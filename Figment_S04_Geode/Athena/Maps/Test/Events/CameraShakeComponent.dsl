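; Actor component that drives a distance-based rocket camera shake for the local player, likely for a test map event.
; On non-authority BeginPlay it zeroes ShakeStrength, enables tick and starts one RocketCameraShake on the first local player's camera manager, keeping the instance.
; Each tick rewrites that shake's Perlin-noise rotation oscillation, scaling amplitude by ShakeStrength and a power falloff of view-target distance from the owner within Radius.
; Zero strength flattens the oscillation.
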(event EventBeginPlay
  (if (not (Networking|HasAuthority (Components|GetOwner)))
    (Variables|Default|SetShakeStrength 0.0)
    (Components|Tick|SetComponentTickEnabled true)
    (bind _output_get (Variables|Default|SetLocalPlayer (Utilities|Array|Get(acopy) (Fortnite|GetLocalFortPlayerControllers true true))))
    (bind _returnvalue (CameraShakes|StartCameraShake :self (Variables|/Script/Engine.PlayerController|GetPlayerCameraManager _output_get) :ShakeClass "/Figment_S04_Geode/Athena/Maps/Test/Events/RocketCameraShake.RocketCameraShake_C" :Scale 1.0 :PlaySpace "CameraLocal"))
    (Variables|Default|SetLocalPlayerCameraShake (CameraShakes|ConvLegacyCameraShake _returnvalue))))

(event EventTick
  (params (DeltaSeconds :type Float%20%28single-precision%29))
  (Utilities|IsValid (Pawn|GetViewTarget (Variables|Default|GetLocalPlayer))
    (:Is%20Valid
      (if (== (Variables|Default|GetShakeStrength) 0.0)
        (bind _localplayercamerashake (Variables|Default|GetLocalPlayerCameraShake)
          (:then
            (Variables|/Script/EngineCameras.LegacyCameraShake|SetRotOscillation (Utilities|Struct|MakeROscillator (struct :Amplitude 0.0 :Frequency 0.0 :Waveform "PerlinNoise") (struct :Amplitude 0.0 :Frequency 0.0 :Waveform "PerlinNoise") (struct :Amplitude 0.0 :Frequency 0.0 :Waveform "PerlinNoise")) _localplayercamerashake))
          (:else))
        (else
          (bind _localplayercamerashake_2 (Variables|Default|GetLocalPlayerCameraShake)
            (:then
              (Variables|/Script/EngineCameras.LegacyCameraShake|SetRotOscillation (Utilities|Struct|MakeROscillator (Utilities|Struct|MakeFOscillator (* (Math|Float|Power (- 1.0 (Math|Float|Clamp(Float) (/ (Math|Vector|VectorLength (- (Transformation|GetActorLocation (Pawn|GetViewTarget (Variables|Default|GetLocalPlayer))) (Transformation|GetActorLocation (Components|GetOwner)))) (Variables|Default|GetRadius)))) (Variables|Default|GetFalloffPower)) (Variables|Default|GetShakeStrength)) (Variables|Default|GetFrequency) "PerlinNoise") (Utilities|Struct|MakeFOscillator (* (Math|Float|Power (- 1.0 (Math|Float|Clamp(Float) (/ (Math|Vector|VectorLength (- (Transformation|GetActorLocation (Pawn|GetViewTarget (Variables|Default|GetLocalPlayer))) (Transformation|GetActorLocation (Components|GetOwner)))) (Variables|Default|GetRadius)))) (Variables|Default|GetFalloffPower)) (Variables|Default|GetShakeStrength)) (Variables|Default|GetFrequency) "PerlinNoise") (Utilities|Struct|MakeFOscillator (* (Math|Float|Power (- 1.0 (Math|Float|Clamp(Float) (/ (Math|Vector|VectorLength (- (Transformation|GetActorLocation (Pawn|GetViewTarget (Variables|Default|GetLocalPlayer))) (Transformation|GetActorLocation (Components|GetOwner)))) (Variables|Default|GetRadius)))) (Variables|Default|GetFalloffPower)) (Variables|Default|GetShakeStrength)) (Variables|Default|GetFrequency) "PerlinNoise")) _localplayercamerashake_2))
            (:else)))))))
