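# Level Sequence in the Geode test event maps that choreographs camera shake and controller rumble over roughly two minutes, with no top-level tracks.
# It animates CameraShake strength, radius and frequency and ForceFeedback intensity on components of a BP_SVLair_GeodeObject2 actor, with rumble peaks largely matching shake spikes.
# A Force Feedback Actor is also toggled by a spawn track in short windows, relocated to a different position for each burst.
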LevelSequence /Figment_S04_Geode/Athena/Maps/Test/Events/CameraShakeShot
bindings [
 Force Feedback Actor {
  id=#1
  possessed_class=ForceFeedbackActor_C
  tracks [
   {
    type=MovieSceneSpawnTrack
    sections [{
      type=MovieSceneSpawnSection
      channels [None {
        type=MovieSceneScriptingBoolChannel
        default=false
        keys [[177,false],[1126,true],[1243.8025,false],[2538,true],[2671.47,false],[2827,true],[2955.4875,false],[3127,true],[3249.7675,false],[3479,true],[3551,false],[3554.6163,true]]
       }]
     }]
   }
   {
    type=MovieScene3DTransformTrack
    sections [{
      type=MovieScene3DTransformSection
      channels [
       Location.X {
        type=MovieSceneScriptingDoubleChannel
        default=0
        thinned_from=6
        keys [[1126,61964.5859],[2538,35021.1367],[2827,-37769.4688],[3127,12555.4854],[3554.6163,12555.4854]]
       }
       Location.Y {
        type=MovieSceneScriptingDoubleChannel
        default=0
        keys [[1126,-110106.7266],[2538,-59284.5234],[2827,46871.957],[3127,-112603.6094],[3479,-112603.6094],[3554.6163,3067.5176]]
       }
       Location.Z {
        type=MovieSceneScriptingDoubleChannel
        default=0
        keys [[1126,71727.5078],[2538,10454.9492],[2827,13817.6318],[3127,14275.5322],[3479,14275.5322],[3554.6163,28563.4961]]
       }
       Rotation.X {type=MovieSceneScriptingDoubleChannel, default=0, const [0,5]}
       Rotation.Y {type=MovieSceneScriptingDoubleChannel, default=0, const [0,5]}
       Rotation.Z {type=MovieSceneScriptingDoubleChannel, default=0, const [0,5]}
       Scale.X {type=MovieSceneScriptingDoubleChannel, default=1, const [1,5]}
       Scale.Y {type=MovieSceneScriptingDoubleChannel, default=1, const [1,5]}
       Scale.Z {type=MovieSceneScriptingDoubleChannel, default=1, const [1,5]}
      ]
     }]
   }
  ]
 }
 Rocket_MainStaticMesh2 {id=#2, possessed_class=BP_SVLair_GeodeObject2_C, tracks []}
 CameraShake {
  id=#3
  parent_id=#2
  possessed_class=CameraShakeComponent_C
  tracks [
   {
    type=MovieSceneFloatTrack
    display_name=Shake Strength
    sections [{
      type=MovieSceneFloatSection
      channels [None {
        type=MovieSceneScriptingFloatChannel
        default=0.5
        keys [[125,0],[152,0.6],[222,0.2],[418,0.3],[651,0.25],[881,0.25],[1137,0.02],[1142,1],[1223,0],[2472,0],[2502,0.5],[2533,0.5],[2538,4],[2564,0.2],[2590,0.5],[2667,0],[2785,0],[2820,0.5],[2824,4],[2839,1],[3124,1],[3127,3],[3182,0],[3433,0],[3471,0.5],[3479,0],[3480,3],[3523,0.1],[3554,1],[3555,7],[3571,3],[3638,0]]
       }]
     }]
   }
   {
    type=MovieSceneFloatTrack
    display_name=Radius
    sections [{
      type=MovieSceneFloatSection
      channels [None {
        type=MovieSceneScriptingFloatChannel
        default=300000
        thinned_from=14
        keys [[0,150000],[222,150000],[392.3125,75000],[1085,75000],[1126,170000],[2429,170000],[2453,300000],[2824,300000],[2839,120000],[3124,120000],[3125,250000],[3541,250000],[3571,300000]]
       }]
     }]
   }
   {
    type=MovieSceneFloatTrack
    display_name=Frequency
    sections [{
      type=MovieSceneFloatSection
      channels [None {type=MovieSceneScriptingFloatChannel, default=30, keys [[0,30],[2824,30],[2839,40],[3118,40],[3122,30]]}]
     }]
   }
  ]
 }
 ForceFeedback {
  id=#4
  parent_id=#2
  possessed_class=ForceFeedbackComponent
  tracks [{
    type=MovieSceneFloatTrack
    display_name=Intensity Multiplier
    sections [{
      type=MovieSceneFloatSection
      channels [None {
        type=MovieSceneScriptingFloatChannel
        default=1
        thinned_from=30
        keys [[7,0],[125,0.1751],[152,0.75],[222,0.4522],[450,0.5],[762,0.05],[1137,0],[1142,1],[1258,0],[2453,0],[2472,1],[2538,1],[2590,0],[2768,0],[2824,0.2],[2839,1],[2870,0.75],[3124,0.75],[3127,1],[3178,0],[3479,0],[3480,1],[3508,0.25],[3522,0.25],[3523,0.7],[3555,1],[3571,0.5],[3638,0]]
       }]
     }]
   }]
 }
]
display_rate {denominator=1, numerator=30}
tick_resolution {denominator=1, numerator=24000}
tracks []
