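# Level sequence for the Geode rocket event in an Athena test map: a staged rocket launch, booster drop, alien-engine phase and crack effects ending in a spawned crack.
# Spawn, visibility, transform and attach tracks time engine emitters, spline smoke trails, a light and the falling booster, swapped at impact for wreck props.
# Particle and material parameter tracks shape exhaust, heat shield and smoke, while crack enter and exit emitters mark teleport hops between booster bursts.
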LevelSequence /Figment_S04_Geode/Athena/Maps/Test/Events/LevelSequence_Geode
bindings [
 SM INverted Sphere {
  id=#1
  possessed_class=FortStaticMeshActor
  tracks [
   {
    type=MovieScene3DTransformTrack
    sections [{
      type=MovieScene3DTransformSection
      channels [
       Location.X {type=MovieSceneScriptingDoubleChannel, default=57600}
       Location.Y {type=MovieSceneScriptingDoubleChannel, default=-114944}
       Location.Z {type=MovieSceneScriptingDoubleChannel, default=8993}
       Rotation.X {type=MovieSceneScriptingDoubleChannel, default=0}
       Rotation.Y {type=MovieSceneScriptingDoubleChannel, default=0}
       Rotation.Z {type=MovieSceneScriptingDoubleChannel, default=0}
       Scale.X {type=MovieSceneScriptingDoubleChannel, default=1}
       Scale.Y {type=MovieSceneScriptingDoubleChannel, default=1}
       Scale.Z {type=MovieSceneScriptingDoubleChannel, default=1}
      ]
     }]
   }
   {
    type=MovieSceneSpawnTrack
    sections [{type=MovieSceneSpawnSection, channels [None {type=MovieSceneScriptingBoolChannel, default=false, keys [[0,false],[1761,true],[4600,false]]}]}]
   }
  ]
 }
 S_RocketBoosterFallAway2 {
  id=#2
  possessed_class=FortStaticMeshActor
  tracks [
   {
    type=MovieSceneVisibilityTrack
    sections [{type=MovieSceneVisibilitySection, channels [None {type=MovieSceneScriptingBoolChannel, keys [[0,false],[2937.3663,true],[3172,false]]}]}]
   }
   {
    type=MovieSceneAudioTrack
    display_name=Audio
    sections [
     {
      type=MovieSceneAudioSection
      end_frame=3172
      start_frame=2933
      channels [
       Volume {type=MovieSceneScriptingFloatChannel, default=1}
       Pitch {type=MovieSceneScriptingFloatChannel, default=1}
      ]
     }
     {
      type=MovieSceneAudioSection
      end_frame=3327
      start_frame=3172
      channels [
       Volume {type=MovieSceneScriptingFloatChannel, default=1}
       Pitch {type=MovieSceneScriptingFloatChannel, default=1}
      ]
     }
     {
      type=MovieSceneAudioSection
      end_frame=3262
      start_frame=3172
      channels [
       Volume {type=MovieSceneScriptingFloatChannel, default=1}
       Pitch {type=MovieSceneScriptingFloatChannel, default=1}
      ]
     }
    ]
   }
   {type=MovieSceneSpawnTrack, sections [{type=MovieSceneSpawnSection, channels [None {type=MovieSceneScriptingBoolChannel, default=true}]}]}
   {
    type=MovieScene3DTransformTrack
    sections [
     {
      type=MovieScene3DTransformSection
      channels [
       Location.X {type=MovieSceneScriptingDoubleChannel, default=61657.6328, keys [[2926,61657.6328],[2937,61657.6328],[3172,108191.9688]]}
       Location.Y {type=MovieSceneScriptingDoubleChannel, default=-110439.4609, keys [[2926,-110439.4609],[2937,-110439.4609],[3172,-24608.0234]]}
       Location.Z {type=MovieSceneScriptingDoubleChannel, default=70127.3047, keys [[2926,70127.3047],[2937,70127.3047],[3172,754.5312]]}
       Rotation.X {type=MovieSceneScriptingDoubleChannel, default=9.2015, keys [[2926,9.2015],[2937,9.2015],[3122,-180],[3172,-180]]}
       Rotation.Y {type=MovieSceneScriptingDoubleChannel, default=4.985, keys [[2926,4.985],[2937,4.985],[3122,-19.9403],[3172,-19.9402]]}
       Rotation.Z {type=MovieSceneScriptingDoubleChannel, default=-47.1916, keys [[2926,-47.1916],[2937,-47.1916],[3122,-115],[3172,-115]]}
       Scale.X {type=MovieSceneScriptingDoubleChannel, default=1, const [1,3]}
       Scale.Y {type=MovieSceneScriptingDoubleChannel, default=1, const [1,3]}
       Scale.Z {type=MovieSceneScriptingDoubleChannel, default=1, const [1,3]}
      ]
     }
     {
      type=MovieScene3DTransformSection
      end_frame=3172
      start_frame=2926
      channels [
       Location.X {type=MovieSceneScriptingDoubleChannel, default=115.4668}
       Location.Y {type=MovieSceneScriptingDoubleChannel, default=-192.3288}
       Location.Z {type=MovieSceneScriptingDoubleChannel, default=0}
       Rotation.X {type=MovieSceneScriptingDoubleChannel, default=0, keys [[2926,9.2015],[2952,-140.2733],[3122,-360]]}
       Rotation.Y {type=MovieSceneScriptingDoubleChannel, default=0, keys [[2926,4.985],[2952,10.163],[3122,0]]}
       Rotation.Z {type=MovieSceneScriptingDoubleChannel, default=0, keys [[2926,-47.1916],[2952,-308.5919],[3122,-720]]}
       Scale.X {type=MovieSceneScriptingDoubleChannel, default=0}
       Scale.Y {type=MovieSceneScriptingDoubleChannel, default=0}
       Scale.Z {type=MovieSceneScriptingDoubleChannel, default=0}
      ]
     }
    ]
   }
  ]
 }
 1x 1Plane on Z {
  id=#3
  possessed_class=FortStaticMeshActor
  tracks [
   {
    type=MovieScene3DTransformTrack
    sections [{
      type=MovieScene3DTransformSection
      channels [
       Location.X {type=MovieSceneScriptingDoubleChannel, default=97.9558, const [-98.0227,1]}
       Location.Y {type=MovieSceneScriptingDoubleChannel, default=-98.018, const [98.1248,1]}
       Location.Z {type=MovieSceneScriptingDoubleChannel, default=3020.9512, const [3021.0698,1]}
       Rotation.X {type=MovieSceneScriptingDoubleChannel, default=0, const [0,2]}
       Rotation.Y {type=MovieSceneScriptingDoubleChannel, default=90, keys [[4112,90],[4230,0]]}
       Rotation.Z {type=MovieSceneScriptingDoubleChannel, default=0, const [0,2]}
       Scale.X {type=MovieSceneScriptingDoubleChannel, default=100, const [100,1]}
       Scale.Y {type=MovieSceneScriptingDoubleChannel, default=100, const [100,1]}
       Scale.Z {type=MovieSceneScriptingDoubleChannel, default=250000, const [250000,1]}
      ]
     }]
   }
   {
    type=MovieSceneSpawnTrack
    sections [{type=MovieSceneSpawnSection, channels [None {type=MovieSceneScriptingBoolChannel, default=false, keys [[0,false],[4116,true],[4332,false]]}]}]
   }
   {type=MovieScene3DAttachTrack, sections [{type=MovieScene3DAttachSection, end_frame=4810, start_frame=4045}]}
  ]
 }
 Alien Booster D {
  id=#4
  possessed_class=Emitter
  tracks [
   {
    type=MovieSceneSpawnTrack
    sections [{
      type=MovieSceneSpawnSection
      channels [None {type=MovieSceneScriptingBoolChannel, default=false, keys [[0,false],[4262.0013,true],[5400,false]]}]
     }]
   }
   {
    type=MovieScene3DTransformTrack
    sections [{
      type=MovieScene3DTransformSection
      channels [
       Location.X {type=MovieSceneScriptingDoubleChannel, default=0}
       Location.Y {type=MovieSceneScriptingDoubleChannel, default=0}
       Location.Z {type=MovieSceneScriptingDoubleChannel, default=0}
       Rotation.X {type=MovieSceneScriptingDoubleChannel, default=0}
       Rotation.Y {type=MovieSceneScriptingDoubleChannel, default=0}
       Rotation.Z {type=MovieSceneScriptingDoubleChannel, default=0}
       Scale.X {type=MovieSceneScriptingDoubleChannel, default=1}
       Scale.Y {type=MovieSceneScriptingDoubleChannel, default=1}
       Scale.Z {type=MovieSceneScriptingDoubleChannel, default=1}
      ]
     }]
   }
   {type=MovieScene3DAttachTrack, sections [{type=MovieScene3DAttachSection, end_frame=5400, start_frame=4262}]}
   {
    type=MovieSceneParticleTrack
    display_name=Particle System
    sections [{
      type=MovieSceneParticleSection
      channels [None {
        type=MovieSceneScriptingParticleChannel
        keys [
         [4262.0013, "<ParticleKey.ACTIVATE: 0>"]
         [4487.0013, "<ParticleKey.DEACTIVATE: 1>"]
         [4629, "<ParticleKey.ACTIVATE: 0>"]
         [4938.9, "<ParticleKey.DEACTIVATE: 1>"]
         [5290, "<ParticleKey.ACTIVATE: 0>"]
         [5358.5, "<ParticleKey.DEACTIVATE: 1>"]
        ]
       }]
     }]
   }
   {
    type=MovieSceneParticleParameterTrack
    display_name=Particle Parameter
    sections [{
      type=MovieSceneParameterSection
      end_frame=5870
      start_frame=4120
      channels [
       BoosterOpacity {type=MovieSceneScriptingFloatChannel, keys [[4262.0025,0],[4287.0025,1]]}
       BoosterSize.X {type=MovieSceneScriptingFloatChannel, default=1, const [1,2]}
       BoosterSize.Y {type=MovieSceneScriptingFloatChannel, default=1, const [1,2]}
       BoosterSize.Z {type=MovieSceneScriptingFloatChannel, default=1, const [1,2]}
       BoosterSize_Sprite.X {type=MovieSceneScriptingFloatChannel, const [1,2]}
       BoosterSize_Sprite.Y {type=MovieSceneScriptingFloatChannel, const [1,2]}
       BoosterSize_Sprite.Z {type=MovieSceneScriptingFloatChannel, const [1,2]}
      ]
     }]
   }
  ]
 }
 P Geode 01 Top Trail {
  id=#5
  possessed_class=Emitter
  tracks [
   {
    type=MovieSceneSpawnTrack
    sections [{type=MovieSceneSpawnSection, channels [None {type=MovieSceneScriptingBoolChannel, default=false, keys [[0,false],[1722,true],[3586,false]]}]}]
   }
   {
    type=MovieScene3DTransformTrack
    sections [{
      type=MovieScene3DTransformSection
      channels [
       Location.X {type=MovieSceneScriptingDoubleChannel, default=57600}
       Location.Y {type=MovieSceneScriptingDoubleChannel, default=-114944}
       Location.Z {type=MovieSceneScriptingDoubleChannel, default=9708}
       Rotation.X {type=MovieSceneScriptingDoubleChannel, default=0}
       Rotation.Y {type=MovieSceneScriptingDoubleChannel, default=0}
       Rotation.Z {type=MovieSceneScriptingDoubleChannel, default=0}
       Scale.X {type=MovieSceneScriptingDoubleChannel, default=1}
       Scale.Y {type=MovieSceneScriptingDoubleChannel, default=1}
       Scale.Z {type=MovieSceneScriptingDoubleChannel, default=1}
      ]
     }]
   }
   {
    type=MovieSceneParticleTrack
    display_name=Particle System
    sections [{type=MovieSceneParticleSection, channels [None {type=MovieSceneScriptingParticleChannel, const ["<ParticleKey.TRIGGER: 2>", 1]}]}]
   }
  ]
 }
 P_Geode_CrackEnter_Tele_2 {
  id=#6
  possessed_class=Emitter
  tracks [
   {
    type=MovieSceneParticleTrack
    display_name=Particle System
    sections [{
      type=MovieSceneParticleSection
      channels [None {type=MovieSceneScriptingParticleChannel, keys [[4916, "<ParticleKey.ACTIVATE: 0>"], [4931, "<ParticleKey.DEACTIVATE: 1>"]]}]
     }]
   }
   {type=MovieSceneSpawnTrack, sections [{type=MovieSceneSpawnSection, channels [None {type=MovieSceneScriptingBoolChannel, default=true}]}]}
   {
    type=MovieScene3DTransformTrack
    sections [{
      type=MovieScene3DTransformSection
      channels [
       Location.X {type=MovieSceneScriptingDoubleChannel, default=12574.3643}
       Location.Y {type=MovieSceneScriptingDoubleChannel, default=-112581.4375}
       Location.Z {type=MovieSceneScriptingDoubleChannel, default=14269.5352}
       Rotation.X {type=MovieSceneScriptingDoubleChannel, default=898.5852}
       Rotation.Y {type=MovieSceneScriptingDoubleChannel, default=9.9544}
       Rotation.Z {type=MovieSceneScriptingDoubleChannel, default=-490.4404}
       Scale.X {type=MovieSceneScriptingDoubleChannel, default=1}
       Scale.Y {type=MovieSceneScriptingDoubleChannel, default=1}
       Scale.Z {type=MovieSceneScriptingDoubleChannel, default=1}
      ]
     }]
   }
  ]
 }
 RocketBits_LowerGreebles {
  id=#7
  parent_id=#8
  possessed_class=StaticMeshComponent
  tracks [{
    type=MovieSceneVisibilityTrack
    sections [{type=MovieSceneVisibilitySection, channels [None {type=MovieSceneScriptingBoolChannel, default=true, const [true,1]}]}]
   }]
 }
 StaticMeshComponent0 {
  id=#9
  parent_id=#10
  possessed_class=StaticMeshComponent
  tracks [{
    type=MovieSceneComponentMaterialTrack
    display_name=Material Element 0
    sections [{
      type=MovieSceneParameterSection
      end_frame=4359
      start_frame=4116
      channels [
       Opacity {type=MovieSceneScriptingFloatChannel, keys [[4116,0],[4230,1]]}
       Multiplier {type=MovieSceneScriptingFloatChannel, default=10, keys [[4249,1],[4259,10],[4301,1]]}
      ]
     }]
   }]
 }
 SplineMeshComponent0 {
  id=#11
  parent_id=#12
  possessed_class=SplineMeshComponent
  tracks [
   {
    type=MovieSceneComponentMaterialTrack
    display_name=Material Element 0
    sections [{
      type=MovieSceneParameterSection
      end_frame=7055
      start_frame=0
      channels [
       PhaseBias {type=MovieSceneScriptingFloatChannel, default=0, const [0,1]}
       Opacity {type=MovieSceneScriptingFloatChannel, default=0, keys [[4393,1],[4632,0]]}
       ScaleV {type=MovieSceneScriptingFloatChannel, default=12, const [4,1]}
       SubSurfaceAmount_1 {type=MovieSceneScriptingFloatChannel, default=0, keys [[2285.7188,0],[2885,0.6]]}
       Bottom Close {type=MovieSceneScriptingFloatChannel, default=1, keys [[1934.1,0],[2953.7188,1]]}
      ]
     }]
   }
   {
    type=MovieScene3DTransformTrack
    sections [{
      type=MovieScene3DTransformSection
      channels [
       Location.X {type=MovieSceneScriptingDoubleChannel, default=0, const [57614.8125,1]}
       Location.Y {type=MovieSceneScriptingDoubleChannel, default=0, const [-114938.9531,1]}
       Location.Z {type=MovieSceneScriptingDoubleChannel, default=0, const [6847,1]}
       Rotation.X {type=MovieSceneScriptingDoubleChannel, default=0, const [0,1]}
       Rotation.Y {type=MovieSceneScriptingDoubleChannel, default=0, const [90,1]}
       Rotation.Z {type=MovieSceneScriptingDoubleChannel, default=0, const [0,1]}
       Scale.X {type=MovieSceneScriptingDoubleChannel, default=1, const [1,1]}
       Scale.Y {type=MovieSceneScriptingDoubleChannel, default=1, const [1,1]}
       Scale.Z {type=MovieSceneScriptingDoubleChannel, default=1, const [1,1]}
      ]
     }]
   }
  ]
 }
 Athena_Prop_Stage3 {
  id=#13
  possessed_class=Athena_Prop_Stage2_C
  tracks [
   {
    type=MovieSceneSpawnTrack
    sections [{type=MovieSceneSpawnSection, channels [None {type=MovieSceneScriptingBoolChannel, default=false, keys [[0,false],[3172,true]]}]}]
   }
   {
    type=MovieScene3DTransformTrack
    sections [{
      type=MovieScene3DTransformSection
      channels [
       Location.X {type=MovieSceneScriptingDoubleChannel, default=108196.8516}
       Location.Y {type=MovieSceneScriptingDoubleChannel, default=-24605.168}
       Location.Z {type=MovieSceneScriptingDoubleChannel, default=753.4294}
       Rotation.X {type=MovieSceneScriptingDoubleChannel, default=165.266}
       Rotation.Y {type=MovieSceneScriptingDoubleChannel, default=0.4661}
       Rotation.Z {type=MovieSceneScriptingDoubleChannel, default=-116.2517}
       Scale.X {type=MovieSceneScriptingDoubleChannel, default=1}
       Scale.Y {type=MovieSceneScriptingDoubleChannel, default=1}
       Scale.Z {type=MovieSceneScriptingDoubleChannel, default=1}
      ]
     }]
   }
  ]
 }
 SplineMeshComponent0 {
  id=#14
  parent_id=#15
  possessed_class=SplineMeshComponent
  tracks [{
    type=MovieSceneComponentMaterialTrack
    display_name=Material Element 0
    sections [{
      type=MovieSceneParameterSection
      end_frame=7055
      start_frame=0
      channels [
       PhaseBias {type=MovieSceneScriptingFloatChannel, default=-1, const [-1,1]}
       Opacity {type=MovieSceneScriptingFloatChannel, default=0, keys [[4393,1],[4632,0]]}
       ScaleV {type=MovieSceneScriptingFloatChannel, default=128, const [4,1]}
      ]
     }]
   }]
 }
 Stage One Booster C {
  id=#16
  possessed_class=Emitter
  tracks [
   {
    type=MovieSceneSpawnTrack
    sections [{type=MovieSceneSpawnSection, channels [None {type=MovieSceneScriptingBoolChannel, default=false, keys [[0,false],[1500,true],[3000,false]]}]}]
   }
   {
    type=MovieScene3DTransformTrack
    sections [{
      type=MovieScene3DTransformSection
      channels [
       Location.X {type=MovieSceneScriptingDoubleChannel, default=0}
       Location.Y {type=MovieSceneScriptingDoubleChannel, default=0}
       Location.Z {type=MovieSceneScriptingDoubleChannel, default=0}
       Rotation.X {type=MovieSceneScriptingDoubleChannel, default=0}
       Rotation.Y {type=MovieSceneScriptingDoubleChannel, default=0}
       Rotation.Z {type=MovieSceneScriptingDoubleChannel, default=0}
       Scale.X {type=MovieSceneScriptingDoubleChannel, default=1}
       Scale.Y {type=MovieSceneScriptingDoubleChannel, default=1}
       Scale.Z {type=MovieSceneScriptingDoubleChannel, default=1}
      ]
     }]
   }
   {type=MovieScene3DAttachTrack, sections [{type=MovieScene3DAttachSection, end_frame=3000, start_frame=1500}]}
   {
    type=MovieSceneParticleTrack
    display_name=Particle System
    sections [{
      type=MovieSceneParticleSection
      channels [None {type=MovieSceneScriptingParticleChannel, keys [[1930, "<ParticleKey.ACTIVATE: 0>"], [2885, "<ParticleKey.DEACTIVATE: 1>"]]}]
     }]
   }
  ]
 }
 P_Geode_1_Vents_Smoke2 {
  id=#17
  possessed_class=Emitter
  tracks [
   {
    type=MovieSceneParticleTrack
    display_name=Particle System
    sections [{
      type=MovieSceneParticleSection
      channels [None {type=MovieSceneScriptingParticleChannel, keys [[21, "<ParticleKey.ACTIVATE: 0>"], [1937, "<ParticleKey.DEACTIVATE: 1>"]]}]
     }]
   }
   {type=MovieSceneSpawnTrack, sections [{type=MovieSceneSpawnSection, channels [None {type=MovieSceneScriptingBoolChannel, default=true}]}]}
   {
    type=MovieScene3DTransformTrack
    sections [{
      type=MovieScene3DTransformSection
      channels [
       Location.X {type=MovieSceneScriptingDoubleChannel, default=57600}
       Location.Y {type=MovieSceneScriptingDoubleChannel, default=-114944}
       Location.Z {type=MovieSceneScriptingDoubleChannel, default=7368}
       Rotation.X {type=MovieSceneScriptingDoubleChannel, default=0}
       Rotation.Y {type=MovieSceneScriptingDoubleChannel, default=0}
       Rotation.Z {type=MovieSceneScriptingDoubleChannel, default=0}
       Scale.X {type=MovieSceneScriptingDoubleChannel, default=1}
       Scale.Y {type=MovieSceneScriptingDoubleChannel, default=1}
       Scale.Z {type=MovieSceneScriptingDoubleChannel, default=1}
      ]
     }]
   }
  ]
 }
 SplineMeshActor2 {
  id=#18
  display_name=Stage Two Smoke Trail B
  possessed_class=SplineMeshActor
  tracks [
   {
    type=MovieScene3DTransformTrack
    sections [{
      type=MovieScene3DTransformSection
      channels [
       Location.X {type=MovieSceneScriptingDoubleChannel, default=79452}
       Location.Y {type=MovieSceneScriptingDoubleChannel, default=-90762.0859}
       Location.Z {type=MovieSceneScriptingDoubleChannel, default=231862}
       Rotation.X {type=MovieSceneScriptingDoubleChannel, default=0}
       Rotation.Y {type=MovieSceneScriptingDoubleChannel, default=90}
       Rotation.Z {type=MovieSceneScriptingDoubleChannel, default=0}
       Scale.X {type=MovieSceneScriptingDoubleChannel, default=1}
       Scale.Y {type=MovieSceneScriptingDoubleChannel, default=1}
       Scale.Z {type=MovieSceneScriptingDoubleChannel, default=1}
      ]
     }]
   }
   {type=MovieSceneSpawnTrack, sections [{type=MovieSceneSpawnSection, channels [None {type=MovieSceneScriptingBoolChannel, default=true}]}]}
  ]
 }
 MissileSilo_Stage2_Lower {
  id=#19
  parent_id=#8
  possessed_class=StaticMeshComponent
  tracks [{
    type=MovieSceneVisibilityTrack
    sections [{type=MovieSceneVisibilitySection, channels [None {type=MovieSceneScriptingBoolChannel, keys [[0,true],[2932,false]]}]}]
   }]
 }
 P Geode Resident Sparks {
  id=#20
  display_name=Alien Residue Heat Shield
  possessed_class=Emitter
  tracks [
   {
    type=MovieSceneSpawnTrack
    sections [{type=MovieSceneSpawnSection, channels [None {type=MovieSceneScriptingBoolChannel, default=false, keys [[0,false],[4200,true],[5400,false]]}]}]
   }
   {
    type=MovieScene3DTransformTrack
    sections [{
      type=MovieScene3DTransformSection
      channels [
       Location.X {type=MovieSceneScriptingDoubleChannel, default=0}
       Location.Y {type=MovieSceneScriptingDoubleChannel, default=0}
       Location.Z {type=MovieSceneScriptingDoubleChannel, default=0}
       Rotation.X {type=MovieSceneScriptingDoubleChannel, default=0}
       Rotation.Y {type=MovieSceneScriptingDoubleChannel, default=0}
       Rotation.Z {type=MovieSceneScriptingDoubleChannel, default=0}
       Scale.X {type=MovieSceneScriptingDoubleChannel, default=1}
       Scale.Y {type=MovieSceneScriptingDoubleChannel, default=1}
       Scale.Z {type=MovieSceneScriptingDoubleChannel, default=1}
      ]
     }]
   }
   {type=MovieScene3DAttachTrack, sections [{type=MovieScene3DAttachSection, end_frame=5400, start_frame=4200}]}
   {
    type=MovieSceneParticleTrack
    display_name=Particle System
    sections [{
      type=MovieSceneParticleSection
      channels [None {
        type=MovieSceneScriptingParticleChannel
        keys [
         [4262.0013, "<ParticleKey.ACTIVATE: 0>"]
         [4487.0013, "<ParticleKey.DEACTIVATE: 1>"]
         [4629, "<ParticleKey.ACTIVATE: 0>"]
         [4938.9, "<ParticleKey.DEACTIVATE: 1>"]
         [5290, "<ParticleKey.ACTIVATE: 0>"]
         [5358.5, "<ParticleKey.DEACTIVATE: 1>"]
        ]
       }]
     }]
   }
   {
    type=MovieSceneParticleParameterTrack
    display_name=Particle Parameter
    sections [{
      type=MovieSceneParameterSection
      end_frame=5411
      start_frame=4262
      channels [
       Bloom {type=MovieSceneScriptingFloatChannel, default=2, keys [[4321,8],[4344,2],[4628,2],[4640,8],[4931,2],[5290,2],[5296,8],[5322,2]]}
       Fade {type=MovieSceneScriptingFloatChannel, default=0, keys [[4321,0],[4334,1],[4628,1],[4640,0.5],[4802,1],[5290,1]]}
      ]
     }]
   }
  ]
 }
 P_Geode_CrackExit_1 {
  id=#21
  possessed_class=Emitter
  tracks [
   {
    type=MovieSceneParticleTrack
    display_name=Particle System
    sections [{type=MovieSceneParticleSection, channels [None {type=MovieSceneScriptingParticleChannel, const ["<ParticleKey.TRIGGER: 2>", 1]}]}]
   }
   {type=MovieSceneSpawnTrack, sections [{type=MovieSceneSpawnSection, channels [None {type=MovieSceneScriptingBoolChannel, default=true}]}]}
   {
    type=MovieScene3DTransformTrack
    sections [{
      type=MovieScene3DTransformSection
      channels [
       Location.X {type=MovieSceneScriptingDoubleChannel, default=-44968.4219}
       Location.Y {type=MovieSceneScriptingDoubleChannel, default=46873.6445}
       Location.Z {type=MovieSceneScriptingDoubleChannel, default=14658.3398}
       Rotation.X {type=MovieSceneScriptingDoubleChannel, default=908.5438}
       Rotation.Y {type=MovieSceneScriptingDoubleChannel, default=-3.6737}
       Rotation.Z {type=MovieSceneScriptingDoubleChannel, default=-720.5879}
       Scale.X {type=MovieSceneScriptingDoubleChannel, default=1}
       Scale.Y {type=MovieSceneScriptingDoubleChannel, default=1}
       Scale.Z {type=MovieSceneScriptingDoubleChannel, default=1}
      ]
     }]
   }
  ]
 }
 1x 1Plane on Z {
  id=#22
  possessed_class=FortStaticMeshActor
  tracks [
   {
    type=MovieScene3DTransformTrack
    sections [{
      type=MovieScene3DTransformSection
      channels [
       Location.X {type=MovieSceneScriptingDoubleChannel, default=97.9558, const [97.9558,1]}
       Location.Y {type=MovieSceneScriptingDoubleChannel, default=-98.018, const [-98.018,1]}
       Location.Z {type=MovieSceneScriptingDoubleChannel, default=3020.9512, const [3020.9512,1]}
       Rotation.X {type=MovieSceneScriptingDoubleChannel, default=0, const [0,2]}
       Rotation.Y {type=MovieSceneScriptingDoubleChannel, default=0, keys [[4116,-90],[4230,0]]}
       Rotation.Z {type=MovieSceneScriptingDoubleChannel, default=0, const [0,2]}
       Scale.X {type=MovieSceneScriptingDoubleChannel, default=100, const [100,1]}
       Scale.Y {type=MovieSceneScriptingDoubleChannel, default=100, const [100,1]}
       Scale.Z {type=MovieSceneScriptingDoubleChannel, default=250000, const [250000,1]}
      ]
     }]
   }
   {
    type=MovieSceneSpawnTrack
    sections [{type=MovieSceneSpawnSection, channels [None {type=MovieSceneScriptingBoolChannel, default=false, keys [[0,false],[4116,true],[4332,false]]}]}]
   }
   {type=MovieScene3DAttachTrack, sections [{type=MovieScene3DAttachSection, end_frame=4810, start_frame=4045}]}
  ]
 }
 P_Geode_CrackExit_Tele_1 {
  id=#23
  possessed_class=Emitter
  tracks [
   {
    type=MovieSceneParticleTrack
    display_name=Particle System
    sections [{
      type=MovieSceneParticleSection
      channels [None {type=MovieSceneScriptingParticleChannel, keys [[4452, "<ParticleKey.ACTIVATE: 0>"], [4632, "<ParticleKey.DEACTIVATE: 1>"]]}]
     }]
   }
   {type=MovieSceneSpawnTrack, sections [{type=MovieSceneSpawnSection, channels [None {type=MovieSceneScriptingBoolChannel, default=true}]}]}
   {
    type=MovieScene3DTransformTrack
    sections [{
      type=MovieScene3DTransformSection
      channels [
       Location.X {type=MovieSceneScriptingDoubleChannel, default=-44968.4219}
       Location.Y {type=MovieSceneScriptingDoubleChannel, default=46873.6445}
       Location.Z {type=MovieSceneScriptingDoubleChannel, default=14658.3398}
       Rotation.X {type=MovieSceneScriptingDoubleChannel, default=908.5438}
       Rotation.Y {type=MovieSceneScriptingDoubleChannel, default=-3.6737}
       Rotation.Z {type=MovieSceneScriptingDoubleChannel, default=-720.5879}
       Scale.X {type=MovieSceneScriptingDoubleChannel, default=1}
       Scale.Y {type=MovieSceneScriptingDoubleChannel, default=1}
       Scale.Z {type=MovieSceneScriptingDoubleChannel, default=1}
      ]
     }]
   }
  ]
 }
 LightComponent0 {
  id=#24
  parent_id=#25
  possessed_class=PointLightComponent
  tracks [
   {
    type=MovieSceneColorTrack
    display_name=Light Color
    sections [{
      type=MovieSceneColorSection
      channels [
       Color.R {type=MovieSceneScriptingFloatChannel, default=1, keys [[4250,1],[4283,0.1]]}
       Color.G {type=MovieSceneScriptingFloatChannel, default=0.9131, keys [[4250,0.9131],[4283,0.9131]]}
       Color.B {type=MovieSceneScriptingFloatChannel, default=0.3864, keys [[4250,0.3864],[4283,1]]}
       Color.A {type=MovieSceneScriptingFloatChannel, default=1, const [1,2]}
      ]
     }]
   }
   {
    type=MovieSceneFloatTrack
    display_name=Attenuation Radius
    sections [{
      type=MovieSceneFloatSection
      channels [None {type=MovieSceneScriptingFloatChannel, default=0, keys [[0,0],[4341,30000],[4348,0],[4629,30000],[4931,0],[5290,30000],[5358.5,0]]}]
     }]
   }
  ]
 }
 Stage Two Main Engine {
  id=#26
  possessed_class=Emitter
  tracks [
   {
    type=MovieSceneSpawnTrack
    sections [{type=MovieSceneSpawnSection, channels [None {type=MovieSceneScriptingBoolChannel, default=false, keys [[0,false],[2932,true],[3600,false]]}]}]
   }
   {
    type=MovieScene3DTransformTrack
    sections [{
      type=MovieScene3DTransformSection
      channels [
       Location.X {type=MovieSceneScriptingDoubleChannel, default=0}
       Location.Y {type=MovieSceneScriptingDoubleChannel, default=0}
       Location.Z {type=MovieSceneScriptingDoubleChannel, default=0}
       Rotation.X {type=MovieSceneScriptingDoubleChannel, default=0}
       Rotation.Y {type=MovieSceneScriptingDoubleChannel, default=0}
       Rotation.Z {type=MovieSceneScriptingDoubleChannel, default=0}
       Scale.X {type=MovieSceneScriptingDoubleChannel, default=1}
       Scale.Y {type=MovieSceneScriptingDoubleChannel, default=1}
       Scale.Z {type=MovieSceneScriptingDoubleChannel, default=1}
      ]
     }]
   }
   {type=MovieScene3DAttachTrack, sections [{type=MovieScene3DAttachSection, end_frame=3600, start_frame=2932}]}
   {
    type=MovieSceneParticleTrack
    display_name=Particle System
    sections [{
      type=MovieSceneParticleSection
      channels [None {type=MovieSceneScriptingParticleChannel, keys [[2932, "<ParticleKey.ACTIVATE: 0>"], [3472, "<ParticleKey.DEACTIVATE: 1>"]]}]
     }]
   }
   {
    type=MovieSceneParticleParameterTrack
    display_name=Particle Parameter
    sections [{
      type=MovieSceneParameterSection
      end_frame=3471
      start_frame=2932
      channels [
       BoosterOpacity {type=MovieSceneScriptingFloatChannel, default=1, keys [[2932,0],[2953.7188,1]]}
       BoosterSize.X {type=MovieSceneScriptingFloatChannel, default=1, keys [[2932,0],[2953.7188,2],[3004.12,1]]}
       BoosterSize.Y {type=MovieSceneScriptingFloatChannel, default=1, keys [[2932,0],[2953.7188,1]]}
       BoosterSize.Z {type=MovieSceneScriptingFloatChannel, default=1, keys [[2932,0],[2953.7188,1]]}
       BoosterSizeSprite.X {type=MovieSceneScriptingFloatChannel, default=0, keys [[2932,0],[2953.7188,1]]}
       BoosterSizeSprite.Y {type=MovieSceneScriptingFloatChannel, default=1, keys [[2932,0],[2953.7188,2],[3004.12,1]]}
       BoosterSizeSprite.Z {type=MovieSceneScriptingFloatChannel, default=1, keys [[2932,0],[2953.7188,1]]}
      ]
     }]
   }
  ]
 }
 SplineMeshActor1 {
  id=#12
  display_name=Stage One Smoke Trail
  possessed_class=SplineMeshActor
  tracks [
   {
    type=MovieScene3DTransformTrack
    sections [{
      type=MovieScene3DTransformSection
      channels [
       Location.X {type=MovieSceneScriptingDoubleChannel, default=57614.8125}
       Location.Y {type=MovieSceneScriptingDoubleChannel, default=-114938.9531}
       Location.Z {type=MovieSceneScriptingDoubleChannel, default=6847}
       Rotation.X {type=MovieSceneScriptingDoubleChannel, default=0}
       Rotation.Y {type=MovieSceneScriptingDoubleChannel, default=90}
       Rotation.Z {type=MovieSceneScriptingDoubleChannel, default=0}
       Scale.X {type=MovieSceneScriptingDoubleChannel, default=1}
       Scale.Y {type=MovieSceneScriptingDoubleChannel, default=1}
       Scale.Z {type=MovieSceneScriptingDoubleChannel, default=1}
      ]
     }]
   }
   {type=MovieSceneSpawnTrack, sections [{type=MovieSceneSpawnSection, channels [None {type=MovieSceneScriptingBoolChannel, default=true}]}]}
  ]
 }
 Booster_PointLight {
  id=#25
  possessed_class=PointLight
  tracks [
   {type=MovieSceneSpawnTrack, sections [{type=MovieSceneSpawnSection, channels [None {type=MovieSceneScriptingBoolChannel, default=true}]}]}
   {type=MovieScene3DAttachTrack, sections [{type=MovieScene3DAttachSection, end_frame=6000, start_frame=0}]}
  ]
 }
 StaticMeshComponent0 {
  id=#27
  parent_id=#22
  possessed_class=StaticMeshComponent
  tracks [{
    type=MovieSceneComponentMaterialTrack
    display_name=Material Element 0
    sections [{
      type=MovieSceneParameterSection
      end_frame=4379
      start_frame=4116
      channels [
       Opacity {type=MovieSceneScriptingFloatChannel, keys [[4116,0],[4230,1]]}
       Multiplier {type=MovieSceneScriptingFloatChannel, default=10, keys [[4249,1],[4259,10],[4301,1]]}
      ]
     }]
   }]
 }
 P_Geode_CrackEnter_2 {
  id=#28
  possessed_class=Emitter
  tracks [
   {
    type=MovieSceneParticleTrack
    display_name=Particle System
    sections [{type=MovieSceneParticleSection, channels [None {type=MovieSceneScriptingParticleChannel, const ["<ParticleKey.TRIGGER: 2>", 1]}]}]
   }
   {type=MovieSceneSpawnTrack, sections [{type=MovieSceneSpawnSection, channels [None {type=MovieSceneScriptingBoolChannel, default=true}]}]}
   {
    type=MovieScene3DTransformTrack
    sections [{
      type=MovieScene3DTransformSection
      channels [
       Location.X {type=MovieSceneScriptingDoubleChannel, default=12574.3643}
       Location.Y {type=MovieSceneScriptingDoubleChannel, default=-112581.4375}
       Location.Z {type=MovieSceneScriptingDoubleChannel, default=14269.5352}
       Rotation.X {type=MovieSceneScriptingDoubleChannel, default=898.5852}
       Rotation.Y {type=MovieSceneScriptingDoubleChannel, default=9.9544}
       Rotation.Z {type=MovieSceneScriptingDoubleChannel, default=-490.4404}
       Scale.X {type=MovieSceneScriptingDoubleChannel, default=1}
       Scale.Y {type=MovieSceneScriptingDoubleChannel, default=1}
       Scale.Z {type=MovieSceneScriptingDoubleChannel, default=1}
      ]
     }]
   }
  ]
 }
 P_Geode_Crack_Spawn {
  id=#29
  display_name=P_Geode_Crack_Spawn_FINAL
  possessed_class=Emitter
  tracks [
   {
    type=MovieSceneParticleTrack
    display_name=Particle System
    sections [{
      type=MovieSceneParticleSection
      channels [None {type=MovieSceneScriptingParticleChannel, keys [[1, "<ParticleKey.DEACTIVATE: 1>"], [5358.1538, "<ParticleKey.TRIGGER: 2>"]]}]
     }]
   }
   {
    type=MovieScene3DTransformTrack
    sections [{
      type=MovieScene3DTransformSection
      channels [
       Location.X {type=MovieSceneScriptingDoubleChannel, default=0, const [5769.481,1]}
       Location.Y {type=MovieSceneScriptingDoubleChannel, default=0, const [5569.2822,1]}
       Location.Z {type=MovieSceneScriptingDoubleChannel, default=0, const [106302.4062,1]}
       Rotation.X {type=MovieSceneScriptingDoubleChannel, default=0, const [0,1]}
       Rotation.Y {type=MovieSceneScriptingDoubleChannel, default=0, const [0,1]}
       Rotation.Z {type=MovieSceneScriptingDoubleChannel, default=0, const [147.6619,1]}
       Scale.X {type=MovieSceneScriptingDoubleChannel, default=1}
       Scale.Y {type=MovieSceneScriptingDoubleChannel, default=1}
       Scale.Z {type=MovieSceneScriptingDoubleChannel, default=1}
      ]
     }]
   }
   {type=MovieSceneSpawnTrack, sections [{type=MovieSceneSpawnSection, channels [None {type=MovieSceneScriptingBoolChannel, default=true}]}]}
  ]
 }
 Alien Booster C {
  id=#30
  possessed_class=Emitter
  tracks [
   {
    type=MovieSceneSpawnTrack
    sections [{
      type=MovieSceneSpawnSection
      channels [None {type=MovieSceneScriptingBoolChannel, default=false, keys [[0,false],[4262.0013,true],[5400,false]]}]
     }]
   }
   {
    type=MovieScene3DTransformTrack
    sections [{
      type=MovieScene3DTransformSection
      channels [
       Location.X {type=MovieSceneScriptingDoubleChannel, default=0}
       Location.Y {type=MovieSceneScriptingDoubleChannel, default=0}
       Location.Z {type=MovieSceneScriptingDoubleChannel, default=0}
       Rotation.X {type=MovieSceneScriptingDoubleChannel, default=0}
       Rotation.Y {type=MovieSceneScriptingDoubleChannel, default=0}
       Rotation.Z {type=MovieSceneScriptingDoubleChannel, default=0}
       Scale.X {type=MovieSceneScriptingDoubleChannel, default=1}
       Scale.Y {type=MovieSceneScriptingDoubleChannel, default=1}
       Scale.Z {type=MovieSceneScriptingDoubleChannel, default=1}
      ]
     }]
   }
   {type=MovieScene3DAttachTrack, sections [{type=MovieScene3DAttachSection, end_frame=5400, start_frame=4262}]}
   {
    type=MovieSceneParticleTrack
    display_name=Particle System
    sections [{
      type=MovieSceneParticleSection
      channels [None {
        type=MovieSceneScriptingParticleChannel
        keys [
         [4262.0013, "<ParticleKey.ACTIVATE: 0>"]
         [4487.0013, "<ParticleKey.DEACTIVATE: 1>"]
         [4629, "<ParticleKey.ACTIVATE: 0>"]
         [4938.9, "<ParticleKey.DEACTIVATE: 1>"]
         [5290, "<ParticleKey.ACTIVATE: 0>"]
         [5358.5, "<ParticleKey.DEACTIVATE: 1>"]
        ]
       }]
     }]
   }
   {
    type=MovieSceneParticleParameterTrack
    display_name=Particle Parameter
    sections [{
      type=MovieSceneParameterSection
      end_frame=5870
      start_frame=4120
      channels [
       BoosterOpacity {type=MovieSceneScriptingFloatChannel, keys [[4262.0013,0],[4287.0013,1]]}
       BoosterSize.X {type=MovieSceneScriptingFloatChannel, default=1, const [1,2]}
       BoosterSize.Y {type=MovieSceneScriptingFloatChannel, default=1, const [1,2]}
       BoosterSize.Z {type=MovieSceneScriptingFloatChannel, default=1, const [1,2]}
       BoosterSize_Sprite.X {type=MovieSceneScriptingFloatChannel, const [1,2]}
       BoosterSize_Sprite.Y {type=MovieSceneScriptingFloatChannel, const [1,2]}
       BoosterSize_Sprite.Z {type=MovieSceneScriptingFloatChannel, const [1,2]}
      ]
     }]
   }
  ]
 }
 1x 1Plane on Z {
  id=#31
  possessed_class=FortStaticMeshActor
  tracks [
   {
    type=MovieScene3DTransformTrack
    sections [{
      type=MovieScene3DTransformSection
      channels [
       Location.X {type=MovieSceneScriptingDoubleChannel, default=97.9558, const [97.9819,1]}
       Location.Y {type=MovieSceneScriptingDoubleChannel, default=-98.018, const [98.1384,1]}
       Location.Z {type=MovieSceneScriptingDoubleChannel, default=3020.9512, const [3021.0732,1]}
       Rotation.X {type=MovieSceneScriptingDoubleChannel, default=90, keys [[4112,90],[4230,0]]}
       Rotation.Y {type=MovieSceneScriptingDoubleChannel, default=0, const [0,2]}
       Rotation.Z {type=MovieSceneScriptingDoubleChannel, default=0, const [0,2]}
       Scale.X {type=MovieSceneScriptingDoubleChannel, default=100, const [100,1]}
       Scale.Y {type=MovieSceneScriptingDoubleChannel, default=100, const [100,1]}
       Scale.Z {type=MovieSceneScriptingDoubleChannel, default=250000, const [250000,1]}
      ]
     }]
   }
   {
    type=MovieSceneSpawnTrack
    sections [{type=MovieSceneSpawnSection, channels [None {type=MovieSceneScriptingBoolChannel, default=false, keys [[0,false],[4116,true],[4332,false]]}]}]
   }
   {type=MovieScene3DAttachTrack, sections [{type=MovieScene3DAttachSection, end_frame=4810, start_frame=4045}]}
  ]
 }
 StaticMeshComponent0 {
  id=#32
  parent_id=#1
  possessed_class=StaticMeshComponent
  tracks [
   {
    type=MovieSceneComponentMaterialTrack
    display_name=Material Element 0
    sections [{
      type=MovieSceneParameterSection
      end_frame=4600
      start_frame=1761
      channels [
       Radius {type=MovieSceneScriptingFloatChannel, default=4000, const [4000,1]}
       Opacity {type=MovieSceneScriptingFloatChannel, default=0, keys [[1761,0],[2706,0.5],[4600,0]]}
      ]
     }]
   }
   {
    type=MovieScenePrimitiveMaterialTrack
    display_name=Material Slot: WorldGridMaterial
    sections [{
      type=MovieScenePrimitiveMaterialSection
      channels [None {
        type=MovieSceneScriptingObjectPathChannel
        const ["<Object '/Figment_S04/Effects/Fort_Effects/Materials/M_SmokeGrenadeSphere_Geode' (0x49e27259a00) Class 'Material'>", 1]
       }]
     }]
   }
  ]
 }
 Prop_Rocks_Piles_24 {
  id=#33
  possessed_class=Prop_Rocks_Piles_5_C
  tracks [
   {
    type=MovieSceneSpawnTrack
    sections [{type=MovieSceneSpawnSection, channels [None {type=MovieSceneScriptingBoolChannel, default=false, keys [[0,false],[3172,true]]}]}]
   }
   {
    type=MovieScene3DTransformTrack
    sections [{
      type=MovieScene3DTransformSection
      channels [
       Location.X {type=MovieSceneScriptingDoubleChannel, default=108496}
       Location.Y {type=MovieSceneScriptingDoubleChannel, default=-24512}
       Location.Z {type=MovieSceneScriptingDoubleChannel, default=0}
       Rotation.X {type=MovieSceneScriptingDoubleChannel, default=0}
       Rotation.Y {type=MovieSceneScriptingDoubleChannel, default=0}
       Rotation.Z {type=MovieSceneScriptingDoubleChannel, default=156.2431}
       Scale.X {type=MovieSceneScriptingDoubleChannel, default=1}
       Scale.Y {type=MovieSceneScriptingDoubleChannel, default=1}
       Scale.Z {type=MovieSceneScriptingDoubleChannel, default=1}
      ]
     }]
   }
  ]
 }
 P_Geode_CrackExit_Tele_2 {
  id=#34
  display_name=P_Geode_CrackExit_Tele_2_Lake
  possessed_class=Emitter
  tracks [
   {
    type=MovieSceneParticleTrack
    display_name=Particle System
    sections [{
      type=MovieSceneParticleSection
      channels [None {type=MovieSceneScriptingParticleChannel, keys [[5105, "<ParticleKey.ACTIVATE: 0>"], [5283, "<ParticleKey.DEACTIVATE: 1>"]]}]
     }]
   }
   {type=MovieSceneSpawnTrack, sections [{type=MovieSceneSpawnSection, channels [None {type=MovieSceneScriptingBoolChannel, default=true}]}]}
   {
    type=MovieScene3DTransformTrack
    sections [{
      type=MovieScene3DTransformSection
      channels [
       Location.X {type=MovieSceneScriptingDoubleChannel, default=56164.7578}
       Location.Y {type=MovieSceneScriptingDoubleChannel, default=-45843.6953}
       Location.Z {type=MovieSceneScriptingDoubleChannel, default=5886.0068}
       Rotation.X {type=MovieSceneScriptingDoubleChannel, default=673.5742}
       Rotation.Y {type=MovieSceneScriptingDoubleChannel, default=66.7515}
       Rotation.Z {type=MovieSceneScriptingDoubleChannel, default=-946.9732}
       Scale.X {type=MovieSceneScriptingDoubleChannel, default=1}
       Scale.Y {type=MovieSceneScriptingDoubleChannel, default=1}
       Scale.Z {type=MovieSceneScriptingDoubleChannel, default=1}
      ]
     }]
   }
  ]
 }
 Alien Main Engine {
  id=#35
  possessed_class=Emitter
  tracks [
   {
    type=MovieSceneSpawnTrack
    sections [{
      type=MovieSceneSpawnSection
      channels [None {type=MovieSceneScriptingBoolChannel, default=false, keys [[0,false],[4262.0013,true],[5400,false]]}]
     }]
   }
   {
    type=MovieScene3DTransformTrack
    sections [{
      type=MovieScene3DTransformSection
      channels [
       Location.X {type=MovieSceneScriptingDoubleChannel, default=0, const [0,1]}
       Location.Y {type=MovieSceneScriptingDoubleChannel, default=0, const [0,1]}
       Location.Z {type=MovieSceneScriptingDoubleChannel, default=0, const [0,1]}
       Rotation.X {type=MovieSceneScriptingDoubleChannel, default=0, const [0,1]}
       Rotation.Y {type=MovieSceneScriptingDoubleChannel, default=0, const [0,1]}
       Rotation.Z {type=MovieSceneScriptingDoubleChannel, default=0, const [0,1]}
       Scale.X {type=MovieSceneScriptingDoubleChannel, default=1, const [1,1]}
       Scale.Y {type=MovieSceneScriptingDoubleChannel, default=1, const [1,1]}
       Scale.Z {type=MovieSceneScriptingDoubleChannel, default=1, const [1,1]}
      ]
     }]
   }
   {type=MovieScene3DAttachTrack, sections [{type=MovieScene3DAttachSection, end_frame=5400, start_frame=4262}]}
   {
    type=MovieSceneParticleTrack
    display_name=Particle System
    sections [{
      type=MovieSceneParticleSection
      channels [None {
        type=MovieSceneScriptingParticleChannel
        keys [
         [4262.0013, "<ParticleKey.ACTIVATE: 0>"]
         [4487.0013, "<ParticleKey.DEACTIVATE: 1>"]
         [4629, "<ParticleKey.ACTIVATE: 0>"]
         [4930, "<ParticleKey.DEACTIVATE: 1>"]
         [5290, "<ParticleKey.ACTIVATE: 0>"]
         [5358.5, "<ParticleKey.DEACTIVATE: 1>"]
        ]
       }]
     }]
   }
   {
    type=MovieSceneParticleParameterTrack
    display_name=Particle Parameter
    sections [{
      type=MovieSceneParameterSection
      end_frame=5870
      start_frame=4120
      channels [
       BoosterOpacity {type=MovieSceneScriptingFloatChannel, keys [[4262.0013,0],[4287.0013,1]]}
       RibbonSpawnRate {type=MovieSceneScriptingFloatChannel, default=100, keys [[4262.0013,100],[4343,0],[4629,100],[4929.9,0],[5290,100]]}
       BoosterSize.X {type=MovieSceneScriptingFloatChannel, default=4, thinned_from=6, keys [[4269.0013,1],[5319,1],[5325,4],[5331,1]]}
       BoosterSize.Y {type=MovieSceneScriptingFloatChannel, default=2, thinned_from=6, keys [[4269.0013,1],[5319,1],[5325,2],[5331,1]]}
       BoosterSize.Z {type=MovieSceneScriptingFloatChannel, default=2, thinned_from=6, keys [[4269.0013,1],[5319,1],[5325,2],[5331,1]]}
       BoosterSize_Sprite.X {type=MovieSceneScriptingFloatChannel, const [1,2]}
       BoosterSize_Sprite.Y {type=MovieSceneScriptingFloatChannel, default=2, thinned_from=6, keys [[4269.0013,1],[5290,1],[5319,2],[5325,2],[5331,1]]}
       BoosterSize_Sprite.Z {type=MovieSceneScriptingFloatChannel, const [1,2]}
       CrossSize.X {type=MovieSceneScriptingFloatChannel, default=1, keys [[4269.0013,2],[4404.0013,10],[4487.0013,1],[5290,1],[5319,4]]}
       CrossSize.Y {type=MovieSceneScriptingFloatChannel, default=1, keys [[4269.0013,2],[4404.0013,10],[4487.0013,1],[5290,1],[5319,4]]}
       CrossSize.Z {type=MovieSceneScriptingFloatChannel, default=1, keys [[4269.0013,2],[4404.0013,10],[4487.0013,1],[5290,1],[5319,4]]}
       EndColor.X {type=MovieSceneScriptingFloatChannel, default=1, keys [[4262.0013,1],[5319,1],[5325,5],[5331,1]]}
       EndColor.Y {type=MovieSceneScriptingFloatChannel, default=1, keys [[4262.0013,1],[5319,1],[5325,5],[5331,1]]}
       EndColor.Z {type=MovieSceneScriptingFloatChannel, default=1, keys [[4262.0013,1],[5319,1],[5325,5],[5331,1]]}
      ]
     }]
   }
  ]
 }
 BP_GeodeScripting {
  id=#36
  possessed_class=BP_GeodeScripting_C
  tracks [{
    type=MovieSceneEventTrack
    sections [{
      type=MovieSceneEventTriggerSection
      channels [None {
        type=MovieSceneScriptingEventChannel
        keys [
         [3172, "<Struct 'MovieSceneEvent' (0x49e4042bc20) {}>"]
         [5358, "<Struct 'MovieSceneEvent' (0x49e4042bb80) {}>"]
         [5358.1538, "<Struct 'MovieSceneEvent' (0x49e4042bcc0) {}>"]
        ]
       }]
     }]
   }]
 }
 P Geode 2 Separation {
  id=#37
  possessed_class=Emitter
  tracks [
   {
    type=MovieSceneSpawnTrack
    sections [{type=MovieSceneSpawnSection, channels [None {type=MovieSceneScriptingBoolChannel, default=false, keys [[0,false],[2700,true],[3060,false]]}]}]
   }
   {
    type=MovieScene3DTransformTrack
    sections [{
      type=MovieScene3DTransformSection
      channels [
       Location.X {type=MovieSceneScriptingDoubleChannel, default=0}
       Location.Y {type=MovieSceneScriptingDoubleChannel, default=0}
       Location.Z {type=MovieSceneScriptingDoubleChannel, default=0}
       Rotation.X {type=MovieSceneScriptingDoubleChannel, default=0}
       Rotation.Y {type=MovieSceneScriptingDoubleChannel, default=0}
       Rotation.Z {type=MovieSceneScriptingDoubleChannel, default=0}
       Scale.X {type=MovieSceneScriptingDoubleChannel, default=1}
       Scale.Y {type=MovieSceneScriptingDoubleChannel, default=1}
       Scale.Z {type=MovieSceneScriptingDoubleChannel, default=1}
      ]
     }]
   }
   {
    type=MovieSceneParticleTrack
    display_name=Particle System
    sections [{type=MovieSceneParticleSection, channels [None {type=MovieSceneScriptingParticleChannel, const ["<ParticleKey.TRIGGER: 2>", 1]}]}]
   }
   {type=MovieScene3DAttachTrack, sections [{type=MovieScene3DAttachSection, end_frame=3060, start_frame=2700}]}
  ]
 }
 P_Geode_CrackExit_2 {
  id=#38
  display_name=P_Geode_CrackExit_2_Lake
  possessed_class=Emitter
  tracks [
   {
    type=MovieSceneParticleTrack
    display_name=Particle System
    sections [{type=MovieSceneParticleSection, channels [None {type=MovieSceneScriptingParticleChannel, const ["<ParticleKey.TRIGGER: 2>", 1]}]}]
   }
   {type=MovieSceneSpawnTrack, sections [{type=MovieSceneSpawnSection, channels [None {type=MovieSceneScriptingBoolChannel, default=true}]}]}
   {
    type=MovieScene3DTransformTrack
    sections [{
      type=MovieScene3DTransformSection
      channels [
       Location.X {type=MovieSceneScriptingDoubleChannel, default=56164.7578}
       Location.Y {type=MovieSceneScriptingDoubleChannel, default=-45843.6953}
       Location.Z {type=MovieSceneScriptingDoubleChannel, default=5886.0068}
       Rotation.X {type=MovieSceneScriptingDoubleChannel, default=673.5742}
       Rotation.Y {type=MovieSceneScriptingDoubleChannel, default=66.7515}
       Rotation.Z {type=MovieSceneScriptingDoubleChannel, default=-946.9732}
       Scale.X {type=MovieSceneScriptingDoubleChannel, default=1}
       Scale.Y {type=MovieSceneScriptingDoubleChannel, default=1}
       Scale.Z {type=MovieSceneScriptingDoubleChannel, default=1}
      ]
     }]
   }
  ]
 }
 Stage One Booster D {
  id=#39
  possessed_class=Emitter
  tracks [
   {
    type=MovieSceneSpawnTrack
    sections [{type=MovieSceneSpawnSection, channels [None {type=MovieSceneScriptingBoolChannel, default=false, keys [[0,false],[1500,true],[3000,false]]}]}]
   }
   {
    type=MovieScene3DTransformTrack
    sections [{
      type=MovieScene3DTransformSection
      channels [
       Location.X {type=MovieSceneScriptingDoubleChannel, default=0}
       Location.Y {type=MovieSceneScriptingDoubleChannel, default=0}
       Location.Z {type=MovieSceneScriptingDoubleChannel, default=0}
       Rotation.X {type=MovieSceneScriptingDoubleChannel, default=0}
       Rotation.Y {type=MovieSceneScriptingDoubleChannel, default=0}
       Rotation.Z {type=MovieSceneScriptingDoubleChannel, default=0}
       Scale.X {type=MovieSceneScriptingDoubleChannel, default=1}
       Scale.Y {type=MovieSceneScriptingDoubleChannel, default=1}
       Scale.Z {type=MovieSceneScriptingDoubleChannel, default=1}
      ]
     }]
   }
   {type=MovieScene3DAttachTrack, sections [{type=MovieScene3DAttachSection, end_frame=3000, start_frame=1500}]}
   {
    type=MovieSceneParticleTrack
    display_name=Particle System
    sections [{
      type=MovieSceneParticleSection
      channels [None {type=MovieSceneScriptingParticleChannel, keys [[1930, "<ParticleKey.ACTIVATE: 0>"], [2885, "<ParticleKey.DEACTIVATE: 1>"]]}]
     }]
   }
  ]
 }
 Alien Booster A {
  id=#40
  possessed_class=Emitter
  tracks [
   {
    type=MovieSceneSpawnTrack
    sections [{
      type=MovieSceneSpawnSection
      channels [None {type=MovieSceneScriptingBoolChannel, default=false, keys [[0,false],[4262.0025,true],[5400,false]]}]
     }]
   }
   {
    type=MovieScene3DTransformTrack
    sections [{
      type=MovieScene3DTransformSection
      channels [
       Location.X {type=MovieSceneScriptingDoubleChannel, default=0}
       Location.Y {type=MovieSceneScriptingDoubleChannel, default=0}
       Location.Z {type=MovieSceneScriptingDoubleChannel, default=0}
       Rotation.X {type=MovieSceneScriptingDoubleChannel, default=0}
       Rotation.Y {type=MovieSceneScriptingDoubleChannel, default=0}
       Rotation.Z {type=MovieSceneScriptingDoubleChannel, default=0}
       Scale.X {type=MovieSceneScriptingDoubleChannel, default=1}
       Scale.Y {type=MovieSceneScriptingDoubleChannel, default=1}
       Scale.Z {type=MovieSceneScriptingDoubleChannel, default=1}
      ]
     }]
   }
   {type=MovieScene3DAttachTrack, sections [{type=MovieScene3DAttachSection, end_frame=5400, start_frame=4262}]}
   {
    type=MovieSceneParticleTrack
    display_name=Particle System
    sections [{
      type=MovieSceneParticleSection
      channels [None {
        type=MovieSceneScriptingParticleChannel
        keys [
         [4262.0013, "<ParticleKey.ACTIVATE: 0>"]
         [4487.0013, "<ParticleKey.DEACTIVATE: 1>"]
         [4629, "<ParticleKey.ACTIVATE: 0>"]
         [4938.9, "<ParticleKey.DEACTIVATE: 1>"]
         [5290, "<ParticleKey.ACTIVATE: 0>"]
         [5358.5, "<ParticleKey.DEACTIVATE: 1>"]
        ]
       }]
     }]
   }
   {
    type=MovieSceneParticleParameterTrack
    display_name=Particle Parameter
    sections [{
      type=MovieSceneParameterSection
      end_frame=5870
      start_frame=4119
      channels [
       BoosterOpacity {type=MovieSceneScriptingFloatChannel, default=0, keys [[4269.0025,0],[4294.0025,1]]}
       BoosterSize.X {type=MovieSceneScriptingFloatChannel, default=0, const [1,2]}
       BoosterSize.Y {type=MovieSceneScriptingFloatChannel, default=0, const [1,2]}
       BoosterSize.Z {type=MovieSceneScriptingFloatChannel, default=0, const [1,2]}
       BoosterSize_Sprite.X {type=MovieSceneScriptingFloatChannel, const [1,2]}
       BoosterSize_Sprite.Y {type=MovieSceneScriptingFloatChannel, default=4, const [1,2]}
       BoosterSize_Sprite.Z {type=MovieSceneScriptingFloatChannel, const [1,2]}
      ]
     }]
   }
  ]
 }
 P_Geode_2_Segment_Trail_New {
  id=#41
  possessed_class=Emitter
  tracks [
   {
    type=MovieSceneParticleTrack
    display_name=Particle System
    sections [{
      type=MovieSceneParticleSection
      channels [None {type=MovieSceneScriptingParticleChannel, keys [[2937.3, "<ParticleKey.ACTIVATE: 0>"], [3172, "<ParticleKey.DEACTIVATE: 1>"]]}]
     }]
   }
   {type=MovieSceneSpawnTrack, sections [{type=MovieSceneSpawnSection, channels [None {type=MovieSceneScriptingBoolChannel, default=true}]}]}
   {type=MovieScene3DAttachTrack, sections [{type=MovieScene3DAttachSection, end_frame=6000, start_frame=0}]}
  ]
 }
 Stage One Booster A {
  id=#42
  possessed_class=Emitter
  tracks [
   {
    type=MovieSceneSpawnTrack
    sections [{type=MovieSceneSpawnSection, channels [None {type=MovieSceneScriptingBoolChannel, default=false, keys [[0,false],[1500,true],[3000,false]]}]}]
   }
   {
    type=MovieScene3DTransformTrack
    sections [{
      type=MovieScene3DTransformSection
      channels [
       Location.X {type=MovieSceneScriptingDoubleChannel, default=0}
       Location.Y {type=MovieSceneScriptingDoubleChannel, default=0}
       Location.Z {type=MovieSceneScriptingDoubleChannel, default=0}
       Rotation.X {type=MovieSceneScriptingDoubleChannel, default=0}
       Rotation.Y {type=MovieSceneScriptingDoubleChannel, default=0}
       Rotation.Z {type=MovieSceneScriptingDoubleChannel, default=0}
       Scale.X {type=MovieSceneScriptingDoubleChannel, default=1}
       Scale.Y {type=MovieSceneScriptingDoubleChannel, default=1}
       Scale.Z {type=MovieSceneScriptingDoubleChannel, default=1}
      ]
     }]
   }
   {type=MovieScene3DAttachTrack, sections [{type=MovieScene3DAttachSection, end_frame=3000, start_frame=1500}]}
   {
    type=MovieSceneParticleTrack
    display_name=Particle System
    sections [{
      type=MovieSceneParticleSection
      channels [None {type=MovieSceneScriptingParticleChannel, keys [[1930, "<ParticleKey.ACTIVATE: 0>"], [2885, "<ParticleKey.DEACTIVATE: 1>"]]}]
     }]
   }
  ]
 }
 SplineMeshActor4 {
  id=#43
  display_name=Stage One to Two S Curve
  possessed_class=SplineMeshActor
  tracks [
   {type=MovieSceneSpawnTrack, sections [{type=MovieSceneSpawnSection, channels [None {type=MovieSceneScriptingBoolChannel, default=true}]}]}
   {
    type=MovieScene3DTransformTrack
    sections [{
      type=MovieScene3DTransformSection
      channels [
       Location.X {type=MovieSceneScriptingDoubleChannel, default=59940.3398}
       Location.Y {type=MovieSceneScriptingDoubleChannel, default=-112011.7812}
       Location.Z {type=MovieSceneScriptingDoubleChannel, default=65353.5742}
       Rotation.X {type=MovieSceneScriptingDoubleChannel, default=0}
       Rotation.Y {type=MovieSceneScriptingDoubleChannel, default=90}
       Rotation.Z {type=MovieSceneScriptingDoubleChannel, default=0}
       Scale.X {type=MovieSceneScriptingDoubleChannel, default=1}
       Scale.Y {type=MovieSceneScriptingDoubleChannel, default=1}
       Scale.Z {type=MovieSceneScriptingDoubleChannel, default=1}
      ]
     }]
   }
  ]
 }
 Twinkle Flare {
  id=#44
  possessed_class=Emitter
  tracks [
   {
    type=MovieSceneSpawnTrack
    sections [{type=MovieSceneSpawnSection, channels [None {type=MovieSceneScriptingBoolChannel, default=false, keys [[0,false],[2941,true],[3600,false]]}]}]
   }
   {
    type=MovieScene3DTransformTrack
    sections [{
      type=MovieScene3DTransformSection
      channels [
       Location.X {type=MovieSceneScriptingDoubleChannel, default=0}
       Location.Y {type=MovieSceneScriptingDoubleChannel, default=0}
       Location.Z {type=MovieSceneScriptingDoubleChannel, default=0}
       Rotation.X {type=MovieSceneScriptingDoubleChannel, default=0}
       Rotation.Y {type=MovieSceneScriptingDoubleChannel, default=0}
       Rotation.Z {type=MovieSceneScriptingDoubleChannel, default=0}
       Scale.X {type=MovieSceneScriptingDoubleChannel, default=1}
       Scale.Y {type=MovieSceneScriptingDoubleChannel, default=1}
       Scale.Z {type=MovieSceneScriptingDoubleChannel, default=1}
      ]
     }]
   }
   {
    type=MovieSceneParticleTrack
    display_name=Particle System
    sections [{
      type=MovieSceneParticleSection
      channels [None {type=MovieSceneScriptingParticleChannel, keys [[2941, "<ParticleKey.ACTIVATE: 0>"], [3560, "<ParticleKey.DEACTIVATE: 1>"]]}]
     }]
   }
   {
    type=MovieSceneParticleParameterTrack
    display_name=Particle Parameter
    sections [{
      type=MovieSceneParameterSection
      end_frame=4120
      start_frame=2941
      channels [
       SizeScale.X {type=MovieSceneScriptingFloatChannel, default=1, keys [[2941,5],[3455,1]]}
       SizeScale.Y {type=MovieSceneScriptingFloatChannel, default=1, keys [[2941,5],[3455,1]]}
       SizeScale.Z {type=MovieSceneScriptingFloatChannel, default=1, keys [[2941,5],[3455,1]]}
      ]
     }]
   }
   {type=MovieScene3DAttachTrack, sections [{type=MovieScene3DAttachSection, end_frame=3600, start_frame=2888}]}
  ]
 }
 P Geode Flare Alien {
  id=#45
  possessed_class=Emitter
  tracks [
   {
    type=MovieSceneSpawnTrack
    sections [{type=MovieSceneSpawnSection, channels [None {type=MovieSceneScriptingBoolChannel, default=false, keys [[0,false],[3455,true],[4350,false]]}]}]
   }
   {
    type=MovieScene3DTransformTrack
    sections [{
      type=MovieScene3DTransformSection
      channels [
       Location.X {type=MovieSceneScriptingDoubleChannel, default=0}
       Location.Y {type=MovieSceneScriptingDoubleChannel, default=0}
       Location.Z {type=MovieSceneScriptingDoubleChannel, default=0}
       Rotation.X {type=MovieSceneScriptingDoubleChannel, default=0}
       Rotation.Y {type=MovieSceneScriptingDoubleChannel, default=0}
       Rotation.Z {type=MovieSceneScriptingDoubleChannel, default=0}
       Scale.X {type=MovieSceneScriptingDoubleChannel, default=1}
       Scale.Y {type=MovieSceneScriptingDoubleChannel, default=1}
       Scale.Z {type=MovieSceneScriptingDoubleChannel, default=1}
      ]
     }]
   }
   {
    type=MovieSceneParticleTrack
    display_name=Particle System
    sections [{
      type=MovieSceneParticleSection
      channels [None {type=MovieSceneScriptingParticleChannel, keys [[3455, "<ParticleKey.ACTIVATE: 0>"], [4262.0013, "<ParticleKey.DEACTIVATE: 1>"]]}]
     }]
   }
   {type=MovieScene3DAttachTrack, sections [{type=MovieScene3DAttachSection, end_frame=4350, start_frame=3455}]}
   {
    type=MovieSceneParticleParameterTrack
    display_name=Particle Parameter
    sections [{
      type=MovieSceneParameterSection
      end_frame=4305
      start_frame=3455
      channels [
       Flicker {type=MovieSceneScriptingFloatChannel, default=0.0625, keys [[3600,0.0625],[4246,0.2]]}
       SizeScale.X {type=MovieSceneScriptingFloatChannel, default=5, keys [[3455,1],[4246,5]]}
       SizeScale.Y {type=MovieSceneScriptingFloatChannel, default=5, keys [[3455,1],[4246,5]]}
       SizeScale.Z {type=MovieSceneScriptingFloatChannel, default=5, keys [[3455,1],[4246,5]]}
      ]
     }]
   }
  ]
 }
 SplineMeshComponent0 {
  id=#46
  parent_id=#43
  possessed_class=SplineMeshComponent
  tracks [{
    type=MovieSceneComponentMaterialTrack
    display_name=Material Element 0
    sections [{
      type=MovieSceneParameterSection
      end_frame=5908
      start_frame=0
      channels [
       Opacity {type=MovieSceneScriptingFloatChannel, default=0, keys [[4393,1],[4632,0]]}
       ScaleV {type=MovieSceneScriptingFloatChannel, const [4,1]}
       SubSurfaceAmount_1 {type=MovieSceneScriptingFloatChannel, default=0.6, const [0,1]}
      ]
     }]
   }]
 }
 BP_DestroyStructures_Booster {
  id=#47
  possessed_class=BP_DestroyStructures_MovingVolume_C
  tracks [
   {type=MovieScene3DAttachTrack, sections [{type=MovieScene3DAttachSection, end_frame=6000, start_frame=362}]}
   {type=MovieSceneSpawnTrack, sections [{type=MovieSceneSpawnSection, channels [None {type=MovieSceneScriptingBoolChannel, default=true}]}]}
   {
    type=MovieScene3DTransformTrack
    sections [{
      type=MovieScene3DTransformSection
      channels [
       Location.X {type=MovieSceneScriptingDoubleChannel, default=0}
       Location.Y {type=MovieSceneScriptingDoubleChannel, default=0}
       Location.Z {type=MovieSceneScriptingDoubleChannel, default=0}
       Rotation.X {type=MovieSceneScriptingDoubleChannel, default=44.8159}
       Rotation.Y {type=MovieSceneScriptingDoubleChannel, default=3.3838}
       Rotation.Z {type=MovieSceneScriptingDoubleChannel, default=47.299}
       Scale.X {type=MovieSceneScriptingDoubleChannel, default=1, const [1,1]}
       Scale.Y {type=MovieSceneScriptingDoubleChannel, default=1, const [1,1]}
       Scale.Z {type=MovieSceneScriptingDoubleChannel, default=1, const [1,1]}
      ]
     }]
   }
   {
    type=MovieSceneEventTrack
    sections [{
      type=MovieSceneEventTriggerSection
      channels [None {
        type=MovieSceneScriptingEventChannel
        keys [
         [2937.3663, "<Struct 'MovieSceneEvent' (0x49e4042cd00) {}>"]
         [3155, "<Struct 'MovieSceneEvent' (0x49e4042cbc0) {}>"]
         [3172, "<Struct 'MovieSceneEvent' (0x49e4042cc60) {}>"]
        ]
       }]
     }]
   }
  ]
 }
 P Geode 1 Small Chunks Falling {
  id=#48
  possessed_class=Emitter
  tracks [
   {
    type=MovieSceneSpawnTrack
    sections [{type=MovieSceneSpawnSection, channels [None {type=MovieSceneScriptingBoolChannel, default=false, keys [[0,false],[1761,true],[2186,false]]}]}]
   }
   {
    type=MovieScene3DTransformTrack
    sections [{
      type=MovieScene3DTransformSection
      channels [
       Location.X {type=MovieSceneScriptingDoubleChannel, default=0}
       Location.Y {type=MovieSceneScriptingDoubleChannel, default=0}
       Location.Z {type=MovieSceneScriptingDoubleChannel, default=0}
       Rotation.X {type=MovieSceneScriptingDoubleChannel, default=0}
       Rotation.Y {type=MovieSceneScriptingDoubleChannel, default=0}
       Rotation.Z {type=MovieSceneScriptingDoubleChannel, default=0}
       Scale.X {type=MovieSceneScriptingDoubleChannel, default=1}
       Scale.Y {type=MovieSceneScriptingDoubleChannel, default=1}
       Scale.Z {type=MovieSceneScriptingDoubleChannel, default=1}
      ]
     }]
   }
   {type=MovieScene3DAttachTrack, sections [{type=MovieScene3DAttachSection, end_frame=2186, start_frame=1761}]}
   {
    type=MovieSceneParticleTrack
    display_name=Particle System
    sections [{type=MovieSceneParticleSection, channels [None {type=MovieSceneScriptingParticleChannel, const ["<ParticleKey.TRIGGER: 2>", 1]}]}]
   }
  ]
 }
 Alien Booster B {
  id=#49
  possessed_class=Emitter
  tracks [
   {
    type=MovieSceneSpawnTrack
    sections [{
      type=MovieSceneSpawnSection
      channels [None {type=MovieSceneScriptingBoolChannel, default=false, keys [[0,false],[4262.0013,true],[5400,false]]}]
     }]
   }
   {
    type=MovieScene3DTransformTrack
    sections [{
      type=MovieScene3DTransformSection
      channels [
       Location.X {type=MovieSceneScriptingDoubleChannel, default=0}
       Location.Y {type=MovieSceneScriptingDoubleChannel, default=0}
       Location.Z {type=MovieSceneScriptingDoubleChannel, default=0}
       Rotation.X {type=MovieSceneScriptingDoubleChannel, default=0}
       Rotation.Y {type=MovieSceneScriptingDoubleChannel, default=0}
       Rotation.Z {type=MovieSceneScriptingDoubleChannel, default=0}
       Scale.X {type=MovieSceneScriptingDoubleChannel, default=1}
       Scale.Y {type=MovieSceneScriptingDoubleChannel, default=1}
       Scale.Z {type=MovieSceneScriptingDoubleChannel, default=1}
      ]
     }]
   }
   {type=MovieScene3DAttachTrack, sections [{type=MovieScene3DAttachSection, end_frame=5400, start_frame=4262}]}
   {
    type=MovieSceneParticleTrack
    display_name=Particle System
    sections [{
      type=MovieSceneParticleSection
      channels [None {
        type=MovieSceneScriptingParticleChannel
        keys [
         [4262.0013, "<ParticleKey.ACTIVATE: 0>"]
         [4487.0013, "<ParticleKey.DEACTIVATE: 1>"]
         [4629, "<ParticleKey.ACTIVATE: 0>"]
         [4938.9, "<ParticleKey.DEACTIVATE: 1>"]
         [5290, "<ParticleKey.ACTIVATE: 0>"]
         [5358.5, "<ParticleKey.DEACTIVATE: 1>"]
        ]
       }]
     }]
   }
   {
    type=MovieSceneParticleParameterTrack
    display_name=Particle Parameter
    sections [{
      type=MovieSceneParameterSection
      end_frame=5870
      start_frame=4119
      channels [
       BoosterOpacity {type=MovieSceneScriptingFloatChannel, keys [[4262,0],[4287,1]]}
       BoosterSize.X {type=MovieSceneScriptingFloatChannel, default=1, const [1,2]}
       BoosterSize.Y {type=MovieSceneScriptingFloatChannel, default=1, const [1,2]}
       BoosterSize.Z {type=MovieSceneScriptingFloatChannel, default=1, const [1,2]}
       BoosterSize_Sprite.X {type=MovieSceneScriptingFloatChannel, const [1,2]}
       BoosterSize_Sprite.Y {type=MovieSceneScriptingFloatChannel, const [1,2]}
       BoosterSize_Sprite.Z {type=MovieSceneScriptingFloatChannel, const [1,2]}
      ]
     }]
   }
  ]
 }
 1x 1Plane on Z {
  id=#10
  possessed_class=FortStaticMeshActor
  tracks [
   {
    type=MovieScene3DTransformTrack
    sections [{
      type=MovieScene3DTransformSection
      channels [
       Location.X {type=MovieSceneScriptingDoubleChannel, default=97.9558, const [-98.0105,1]}
       Location.Y {type=MovieSceneScriptingDoubleChannel, default=-98.018, const [-97.9172,1]}
       Location.Z {type=MovieSceneScriptingDoubleChannel, default=3020.9512, const [3021.0181,1]}
       Rotation.X {type=MovieSceneScriptingDoubleChannel, default=-90, keys [[4112,-90],[4230,0]]}
       Rotation.Y {type=MovieSceneScriptingDoubleChannel, default=0, const [0,2]}
       Rotation.Z {type=MovieSceneScriptingDoubleChannel, default=0, const [0,2]}
       Scale.X {type=MovieSceneScriptingDoubleChannel, default=100, const [100,1]}
       Scale.Y {type=MovieSceneScriptingDoubleChannel, default=100, const [100,1]}
       Scale.Z {type=MovieSceneScriptingDoubleChannel, default=250000, const [250000,1]}
      ]
     }]
   }
   {
    type=MovieSceneSpawnTrack
    sections [{type=MovieSceneSpawnSection, channels [None {type=MovieSceneScriptingBoolChannel, default=false, keys [[0,false],[4116,true],[4332,false]]}]}]
   }
   {type=MovieScene3DAttachTrack, sections [{type=MovieScene3DAttachSection, end_frame=4810, start_frame=4045}]}
  ]
 }
 SplineMeshActor2 {
  id=#15
  display_name=Stage Two Smoke Trail A
  possessed_class=SplineMeshActor
  tracks [
   {
    type=MovieScene3DTransformTrack
    sections [{
      type=MovieScene3DTransformSection
      channels [
       Location.X {type=MovieSceneScriptingDoubleChannel, default=60316}
       Location.Y {type=MovieSceneScriptingDoubleChannel, default=-111474.0859}
       Location.Z {type=MovieSceneScriptingDoubleChannel, default=117070}
       Rotation.X {type=MovieSceneScriptingDoubleChannel, default=0}
       Rotation.Y {type=MovieSceneScriptingDoubleChannel, default=90}
       Rotation.Z {type=MovieSceneScriptingDoubleChannel, default=0}
       Scale.X {type=MovieSceneScriptingDoubleChannel, default=1}
       Scale.Y {type=MovieSceneScriptingDoubleChannel, default=1}
       Scale.Z {type=MovieSceneScriptingDoubleChannel, default=1}
      ]
     }]
   }
   {type=MovieSceneSpawnTrack, sections [{type=MovieSceneSpawnSection, channels [None {type=MovieSceneScriptingBoolChannel, default=true}]}]}
  ]
 }
 StaticMeshComponent0 {
  id=#50
  parent_id=#3
  possessed_class=StaticMeshComponent
  tracks [{
    type=MovieSceneComponentMaterialTrack
    display_name=Material Element 0
    sections [{
      type=MovieSceneParameterSection
      end_frame=4451
      start_frame=4116
      channels [
       Opacity {type=MovieSceneScriptingFloatChannel, keys [[4116.0013,0],[4230.0013,1]]}
       Multiplier {type=MovieSceneScriptingFloatChannel, default=10, keys [[4249,1],[4259,10],[4301,1]]}
      ]
     }]
   }]
 }
 P Geode 1 Opening Beam {
  id=#51
  possessed_class=Emitter
  tracks [
   {
    type=MovieSceneSpawnTrack
    sections [{type=MovieSceneSpawnSection, channels [None {type=MovieSceneScriptingBoolChannel, default=false, keys [[0,false],[1650,true],[2357,false]]}]}]
   }
   {
    type=MovieScene3DTransformTrack
    sections [{
      type=MovieScene3DTransformSection
      channels [
       Location.X {type=MovieSceneScriptingDoubleChannel, default=57600}
       Location.Y {type=MovieSceneScriptingDoubleChannel, default=-114944}
       Location.Z {type=MovieSceneScriptingDoubleChannel, default=9727}
       Rotation.X {type=MovieSceneScriptingDoubleChannel, default=0}
       Rotation.Y {type=MovieSceneScriptingDoubleChannel, default=0}
       Rotation.Z {type=MovieSceneScriptingDoubleChannel, default=0}
       Scale.X {type=MovieSceneScriptingDoubleChannel, default=1}
       Scale.Y {type=MovieSceneScriptingDoubleChannel, default=1}
       Scale.Z {type=MovieSceneScriptingDoubleChannel, default=1}
      ]
     }]
   }
   {
    type=MovieSceneParticleTrack
    display_name=Particle System
    sections [{type=MovieSceneParticleSection, channels [None {type=MovieSceneScriptingParticleChannel, const ["<ParticleKey.TRIGGER: 2>", 1]}]}]
   }
  ]
 }
 StaticMeshComponent0 {
  id=#52
  parent_id=#31
  possessed_class=StaticMeshComponent
  tracks [{
    type=MovieSceneComponentMaterialTrack
    display_name=Material Element 0
    sections [{
      type=MovieSceneParameterSection
      end_frame=4353
      start_frame=4116
      channels [
       Opacity {type=MovieSceneScriptingFloatChannel, keys [[4116.0013,0],[4230.0013,1]]}
       Multiplier {type=MovieSceneScriptingFloatChannel, default=10, keys [[4249,1],[4259,10],[4301,1]]}
      ]
     }]
   }]
 }
 P_Geode_CrackEnter_Tilted {
  id=#53
  possessed_class=Emitter
  tracks [
   {
    type=MovieSceneParticleTrack
    display_name=Particle System
    sections [{type=MovieSceneParticleSection, channels [None {type=MovieSceneScriptingParticleChannel, const ["<ParticleKey.TRIGGER: 2>", 1]}]}]
   }
   {type=MovieSceneSpawnTrack, sections [{type=MovieSceneSpawnSection, channels [None {type=MovieSceneScriptingBoolChannel, default=true}]}]}
   {
    type=MovieScene3DTransformTrack
    sections [{
      type=MovieScene3DTransformSection
      channels [
       Location.X {type=MovieSceneScriptingDoubleChannel, default=34819.6602}
       Location.Y {type=MovieSceneScriptingDoubleChannel, default=-59070.6836}
       Location.Z {type=MovieSceneScriptingDoubleChannel, default=8666.4092}
       Rotation.X {type=MovieSceneScriptingDoubleChannel, default=945.3856}
       Rotation.Y {type=MovieSceneScriptingDoubleChannel, default=-80.5835}
       Rotation.Z {type=MovieSceneScriptingDoubleChannel, default=-586.2718}
       Scale.X {type=MovieSceneScriptingDoubleChannel, default=1}
       Scale.Y {type=MovieSceneScriptingDoubleChannel, default=1}
       Scale.Z {type=MovieSceneScriptingDoubleChannel, default=1}
      ]
     }]
   }
  ]
 }
 RocketBits_Booster_SectionWhole {
  id=#54
  parent_id=#8
  possessed_class=StaticMeshComponent
  tracks [{
    type=MovieSceneVisibilityTrack
    sections [{type=MovieSceneVisibilitySection, channels [None {type=MovieSceneScriptingBoolChannel, default=true, const [true,1]}]}]
   }]
 }
 Alien Engine Engage {
  id=#55
  possessed_class=Emitter
  tracks [
   {
    type=MovieSceneSpawnTrack
    sections [{type=MovieSceneSpawnSection, channels [None {type=MovieSceneScriptingBoolChannel, default=false, keys [[0,false],[3807,true],[5850,false]]}]}]
   }
   {
    type=MovieScene3DTransformTrack
    sections [{
      type=MovieScene3DTransformSection
      channels [
       Location.X {type=MovieSceneScriptingDoubleChannel, default=0}
       Location.Y {type=MovieSceneScriptingDoubleChannel, default=0}
       Location.Z {type=MovieSceneScriptingDoubleChannel, default=0}
       Rotation.X {type=MovieSceneScriptingDoubleChannel, default=0}
       Rotation.Y {type=MovieSceneScriptingDoubleChannel, default=0}
       Rotation.Z {type=MovieSceneScriptingDoubleChannel, default=0}
       Scale.X {type=MovieSceneScriptingDoubleChannel, default=1}
       Scale.Y {type=MovieSceneScriptingDoubleChannel, default=1}
       Scale.Z {type=MovieSceneScriptingDoubleChannel, default=1}
      ]
     }]
   }
   {
    type=MovieSceneParticleTrack
    display_name=Particle System
    sections [{type=MovieSceneParticleSection, channels [None {type=MovieSceneScriptingParticleChannel, const ["<ParticleKey.TRIGGER: 2>", 2]}]}]
   }
   {type=MovieScene3DAttachTrack, sections [{type=MovieScene3DAttachSection, end_frame=5850, start_frame=3807}]}
  ]
 }
 Rocket_MainStaticMesh {
  id=#8
  possessed_class=BP_SVLair_GeodeObject2_C
  tracks [
   {
    type=MovieScene3DTransformTrack
    sections [{
      type=MovieScene3DTransformSection
      channels [
       Location.X {
        type=MovieSceneScriptingDoubleChannel
        default=80418
        thinned_from=46
        keys [[0,57600],[2325,57600],[2445,57784.4023],[2581,58116.7852],[2703.9225,58646.875],[2883,59998.1953],[2934,61500],[3360,85582.2031],[3445,90276],[3630,100866.8125],[3728.9662,106821.1641],[3764,107424.1953],[3829.6837,104876],[4098,70509.1641],[4171,60357.9961],[4283,45757.4102],[4345,34160.6094],[4626,-55562.7617],[4652.515,-27516.4043],[4685.515,11935.8223],[4704.515,33414.0078],[4726.515,56901.7109],[4760.1488,85178.2188],[4785.9288,101549.5391],[4804.7162,102464.9453],[4837.1862,89070.1719],[4856.9538,77898.2969],[4872.8163,65820.7578],[4899.515,45167.8398],[4930.98,12574.3643],[4939,5003.9307],[5276.965,59608.7891],[5283,56539.1641],[5330,43050],[5358.1538,9867.9355]]
       }
       Location.Y {
        type=MovieSceneScriptingDoubleChannel
        default=-90314
        thinned_from=46
        keys [[0,-114944],[2445,-114842.3203],[2703.9225,-114054.375],[2849,-112490.3516],[2883,-111963.8203],[2934,-110690],[3270,-90314],[3445,-80106],[3630,-68685.9297],[3728.9662,-61843.4375],[3764,-59326.1016],[3829.6837,-56328],[4171,-66457.8203],[4283,-66120.8828],[4345,-58343.1445],[4626,46871.957],[4652.515,46871.957],[4685.515,47467.7773],[4704.515,52544.4336],[4726.515,60335.9688],[4760.1488,55110.3438],[4785.9288,34411.3281],[4804.7162,6686.7461],[4837.1862,-17769.2188],[4856.9538,-33834.5352],[4872.8163,-47680.582],[4899.515,-74115.0078],[4930.98,-112581.4375],[4939,-121471.4297],[5276.965,-50306.5586],[5283,-46153.5898],[5330,-32100],[5358.1538,3067.7051]]
       }
       Location.Z {
        type=MovieSceneScriptingDoubleChannel
        default=277821
        thinned_from=46
        keys [[0,7296],[1949.0738,7296],[2025,7836],[2100,8280],[2703.9225,36209.5312],[2849,59260],[2883,65290],[2923,68620],[2933,69180.6172],[2938,70310],[3270,277821],[3360,344534.625],[3445,396736],[3583,425930],[3670,421590],[3736,400280],[4098,220639.6094],[4171,181878.5],[4283,100765.5469],[4345,2798.0281],[4626,15500],[4652.515,12876],[4685.515,6831.0791],[4704.515,5076.6392],[4726.515,9976.5361],[4760.1488,10978.1055],[4785.9288,8478.9648],[4804.7162,5998.8633],[4837.1862,4449.5376],[4856.9538,3294.0449],[4872.8163,3429.8008],[4899.515,6120.3306],[4930.98,14269.5352],[4939,16674.3984],[5276.965,-5688.3765],[5283,4747.5391],[5330,36610],[5358.1538,105619.5234]]
       }
       Rotation.X {
        type=MovieSceneScriptingDoubleChannel
        default=9
        thinned_from=46
        keys [[0,0],[2250,0],[2445,2.1669],[2704,6.5067],[2849,4.2608],[2883,7.1899],[2923,13.1542],[2938,14],[3090,9],[3396,6.2054],[3630,12.0025],[3728.9662,70.9517],[3764,92.1855],[3829.6837,121.3062],[4070,-541.7042],[4171,-535.7289],[4283,-179.1172],[4345,179.4441],[4626.515,90],[4652.515,180.0001],[4666,-169.7825],[4685.515,-159.5844],[4704.515,-72.8904],[4726.515,64.5459],[4760.1488,32.6622],[4785.9288,173.6846],[4804.7162,-180.0001],[4814,185.5793],[4837.1862,-176.6858],[4856.9538,12.5658],[4872.8163,0],[4899.515,-6.6158],[4930.98,0],[5250.0975,0],[5276.965,-14.5395],[5282,-5.0532],[5283,-3.8194],[5287,-7.4434],[5309,3.2109],[5322,11.6871],[5336,11.7349],[5358.1538,26.3734]]
       }
       Rotation.Y {
        type=MovieSceneScriptingDoubleChannel
        default=1
        thinned_from=48
        keys [[0,0],[2385,0],[2445,-0.653],[2704,-1.9649],[2849,0.2964],[2883,3.8698],[2923,2.0972],[2938,1.5951],[3090,1],[3396,-3.4902],[3630,-11.8707],[3728.9662,0],[3764,6.0428],[3829.6837,20.8612],[4070,380.3051],[4171,351.7582],[4283,-8.2053],[4345,-7.8762],[4626.515,84.3091],[4652.515,84.0718],[4666,80.3008],[4685.515,79.076],[4704.515,69.8355],[4726.515,66.7388],[4760.1488,88.0891],[4785.9288,85.3664],[4804.7162,87.6211],[4837.1862,86.342],[4872.8163,85.8001],[4899.515,80.6782],[4930.98,71.8092],[5250.0975,71.8092],[5276.965,11.0622],[5282,31.7567],[5283,23.4651],[5287,23.8774],[5309,37.7839],[5322,39.8135],[5336,32.3216],[5358.1538,28.2921]]
       }
       Rotation.Z {
        type=MovieSceneScriptingDoubleChannel
        default=0
        thinned_from=49
        keys [[0,0],[2385,0],[2445,-9.8272],[2704,-22.9925],[2849,-30.1909],[2883,-38.9759],[2923,-47.6222],[2938,-48.1828],[3090,-37.157],[3396,-14.706],[3630,1.8365],[3728.9662,0],[3764,-7.3221],[3829.6837,-22.0613],[4070,-431.5125],[4171,-398.7514],[4283,-38.3366],[4345,-31.0839],[4626.515,-90],[4652.515,0.0001],[4666,11.2],[4685.515,25.5338],[4704.515,127.2569],[4726.515,-112.6796],[4760.1488,163.9663],[4785.9288,267.3273],[4804.7162,-108.4503],[4814,-112.0497],[4837.1862,-119.3097],[4856.9538,63.2433],[4872.8163,53.0801],[4899.515,42.3708],[4930.98,53.0792],[5250.0975,53.0792],[5276.62,-46.9603],[5276.965,-57.5532],[5282,-60.8414],[5287,-57.7147],[5309,-41.8354],[5322,-30.2818],[5336,-22.3628],[5358.1538,-6.9067]]
       }
       Scale.X {type=MovieSceneScriptingDoubleChannel, default=1, const [1,32]}
       Scale.Y {type=MovieSceneScriptingDoubleChannel, default=1, const [1,32]}
       Scale.Z {type=MovieSceneScriptingDoubleChannel, default=1, const [1,32]}
      ]
     }]
   }
   {
    type=MovieSceneVisibilityTrack
    sections [{
      type=MovieSceneVisibilitySection
      channels [None {
        type=MovieSceneScriptingBoolChannel
        default=false
        keys [[0,true],[5,true],[1810.6275,true],[1949.0738,true],[2325,true],[3139.6875,true],[3530.6875,true],[3972.6875,true],[3995.6875,true],[4343,false],[4635,true],[4931,false],[5283,true],[5358.1538,false]]
       }]
     }]
   }
   {type=MovieSceneSpawnTrack, sections [{type=MovieSceneSpawnSection, channels [None {type=MovieSceneScriptingBoolChannel, default=true}]}]}
  ]
 }
 RocketBits_UpperTip {
  id=#56
  parent_id=#8
  possessed_class=StaticMeshComponent
  tracks [{
    type=MovieSceneVisibilityTrack
    sections [{type=MovieSceneVisibilitySection, channels [None {type=MovieSceneScriptingBoolChannel, default=true, const [true,1]}]}]
   }]
 }
 P_Geode_2_GroundImpact {
  id=#57
  possessed_class=Emitter
  tracks [
   {
    type=MovieSceneParticleTrack
    display_name=Particle System
    sections [{type=MovieSceneParticleSection, channels [None {type=MovieSceneScriptingParticleChannel, const ["<ParticleKey.TRIGGER: 2>", 1]}]}]
   }
   {
    type=MovieSceneSpawnTrack
    sections [{type=MovieSceneSpawnSection, channels [None {type=MovieSceneScriptingBoolChannel, default=false, keys [[0,false],[2780,true],[3585,false]]}]}]
   }
   {
    type=MovieScene3DTransformTrack
    sections [{
      type=MovieScene3DTransformSection
      channels [
       Location.X {type=MovieSceneScriptingDoubleChannel, default=108460}
       Location.Y {type=MovieSceneScriptingDoubleChannel, default=-24545}
       Location.Z {type=MovieSceneScriptingDoubleChannel, default=107}
       Rotation.X {type=MovieSceneScriptingDoubleChannel, default=0}
       Rotation.Y {type=MovieSceneScriptingDoubleChannel, default=90}
       Rotation.Z {type=MovieSceneScriptingDoubleChannel, default=0}
       Scale.X {type=MovieSceneScriptingDoubleChannel, default=1}
       Scale.Y {type=MovieSceneScriptingDoubleChannel, default=1}
       Scale.Z {type=MovieSceneScriptingDoubleChannel, default=1}
      ]
     }]
   }
  ]
 }
 SplineMeshComponent0 {
  id=#58
  parent_id=#18
  possessed_class=SplineMeshComponent
  tracks [{
    type=MovieSceneComponentMaterialTrack
    display_name=Material Element 0
    sections [{
      type=MovieSceneParameterSection
      end_frame=7247
      start_frame=0
      channels [
       PhaseBias {type=MovieSceneScriptingFloatChannel, default=-2, const [-2,1]}
       Opacity {type=MovieSceneScriptingFloatChannel, default=0, keys [[4393,1],[4632,0]]}
       ScaleV {type=MovieSceneScriptingFloatChannel, const [4,1]}
       SubSurfaceAmount_1 {type=MovieSceneScriptingFloatChannel, keys [[3440,0.6],[3472,0]]}
      ]
     }]
   }]
 }
 P Geode 01 Base Trail {
  id=#59
  possessed_class=Emitter
  tracks [
   {
    type=MovieSceneSpawnTrack
    sections [{type=MovieSceneSpawnSection, channels [None {type=MovieSceneScriptingBoolChannel, default=false, keys [[0,false],[1761,true],[4600,false]]}]}]
   }
   {
    type=MovieScene3DTransformTrack
    sections [{
      type=MovieScene3DTransformSection
      channels [
       Location.X {type=MovieSceneScriptingDoubleChannel, default=0}
       Location.Y {type=MovieSceneScriptingDoubleChannel, default=0}
       Location.Z {type=MovieSceneScriptingDoubleChannel, default=0}
       Rotation.X {type=MovieSceneScriptingDoubleChannel, default=0}
       Rotation.Y {type=MovieSceneScriptingDoubleChannel, default=0}
       Rotation.Z {type=MovieSceneScriptingDoubleChannel, default=0}
       Scale.X {type=MovieSceneScriptingDoubleChannel, default=1}
       Scale.Y {type=MovieSceneScriptingDoubleChannel, default=1}
       Scale.Z {type=MovieSceneScriptingDoubleChannel, default=1}
      ]
     }]
   }
   {
    type=MovieSceneParticleTrack
    display_name=Particle System
    sections [{type=MovieSceneParticleSection, channels [None {type=MovieSceneScriptingParticleChannel, const ["<ParticleKey.TRIGGER: 2>", 1]}]}]
   }
   {type=MovieScene3DAttachTrack, sections [{type=MovieScene3DAttachSection, end_frame=7091, start_frame=1761}]}
  ]
 }
 Stage One Booster B {
  id=#60
  possessed_class=Emitter
  tracks [
   {
    type=MovieSceneSpawnTrack
    sections [{type=MovieSceneSpawnSection, channels [None {type=MovieSceneScriptingBoolChannel, default=false, keys [[0,false],[1500,true],[3000,false]]}]}]
   }
   {
    type=MovieScene3DTransformTrack
    sections [{
      type=MovieScene3DTransformSection
      channels [
       Location.X {type=MovieSceneScriptingDoubleChannel, default=0}
       Location.Y {type=MovieSceneScriptingDoubleChannel, default=0}
       Location.Z {type=MovieSceneScriptingDoubleChannel, default=0}
       Rotation.X {type=MovieSceneScriptingDoubleChannel, default=0}
       Rotation.Y {type=MovieSceneScriptingDoubleChannel, default=0}
       Rotation.Z {type=MovieSceneScriptingDoubleChannel, default=0}
       Scale.X {type=MovieSceneScriptingDoubleChannel, default=1}
       Scale.Y {type=MovieSceneScriptingDoubleChannel, default=1}
       Scale.Z {type=MovieSceneScriptingDoubleChannel, default=1}
      ]
     }]
   }
   {type=MovieScene3DAttachTrack, sections [{type=MovieScene3DAttachSection, end_frame=3000, start_frame=1500}]}
   {
    type=MovieSceneParticleTrack
    display_name=Particle System
    sections [{
      type=MovieSceneParticleSection
      channels [None {type=MovieSceneScriptingParticleChannel, keys [[1930, "<ParticleKey.ACTIVATE: 0>"], [2885, "<ParticleKey.DEACTIVATE: 1>"]]}]
     }]
   }
  ]
 }
 P_Geode_CrackEnter_Tele_Tilted {
  id=#61
  possessed_class=Emitter
  tracks [
   {
    type=MovieSceneParticleTrack
    display_name=Particle System
    sections [{
      type=MovieSceneParticleSection
      channels [None {type=MovieSceneScriptingParticleChannel, keys [[4326, "<ParticleKey.ACTIVATE: 0>"], [4341, "<ParticleKey.DEACTIVATE: 1>"]]}]
     }]
   }
   {type=MovieSceneSpawnTrack, sections [{type=MovieSceneSpawnSection, channels [None {type=MovieSceneScriptingBoolChannel, default=true}]}]}
   {
    type=MovieScene3DTransformTrack
    sections [{
      type=MovieScene3DTransformSection
      channels [
       Location.X {type=MovieSceneScriptingDoubleChannel, default=34819.6602}
       Location.Y {type=MovieSceneScriptingDoubleChannel, default=-59070.6836}
       Location.Z {type=MovieSceneScriptingDoubleChannel, default=8666.4092}
       Rotation.X {type=MovieSceneScriptingDoubleChannel, default=945.3856}
       Rotation.Y {type=MovieSceneScriptingDoubleChannel, default=-80.5835}
       Rotation.Z {type=MovieSceneScriptingDoubleChannel, default=-586.2718}
       Scale.X {type=MovieSceneScriptingDoubleChannel, default=1}
       Scale.Y {type=MovieSceneScriptingDoubleChannel, default=1}
       Scale.Z {type=MovieSceneScriptingDoubleChannel, default=1}
      ]
     }]
   }
  ]
 }
 BP_Geode_DestroyStructures_MovingVolume {
  id=#62
  possessed_class=BP_DestroyStructures_MovingVolume_C
  tracks [
   {type=MovieScene3DAttachTrack, sections [{type=MovieScene3DAttachSection, end_frame=6000, start_frame=362}]}
   {type=MovieSceneSpawnTrack, sections [{type=MovieSceneSpawnSection, channels [None {type=MovieSceneScriptingBoolChannel, default=true}]}]}
   {
    type=MovieScene3DTransformTrack
    sections [{
      type=MovieScene3DTransformSection
      channels [
       Location.X {type=MovieSceneScriptingDoubleChannel, default=0.0039}
       Location.Y {type=MovieSceneScriptingDoubleChannel, default=0}
       Location.Z {type=MovieSceneScriptingDoubleChannel, default=3063.9854}
       Rotation.X {type=MovieSceneScriptingDoubleChannel, default=0.0001}
       Rotation.Y {type=MovieSceneScriptingDoubleChannel, default=0}
       Rotation.Z {type=MovieSceneScriptingDoubleChannel, default=0}
       Scale.X {type=MovieSceneScriptingDoubleChannel, default=1}
       Scale.Y {type=MovieSceneScriptingDoubleChannel, default=1}
       Scale.Z {type=MovieSceneScriptingDoubleChannel, default=1}
      ]
     }]
   }
   {
    type=MovieSceneEventTrack
    sections [{
      type=MovieSceneEventTriggerSection
      channels [None {
        type=MovieSceneScriptingEventChannel
        keys [
         [1921.8813, "<Struct 'MovieSceneEvent' (0x49e4042da20) {}>"]
         [2402.985, "<Struct 'MovieSceneEvent' (0x49e4042d980) {}>"]
         [4301, "<Struct 'MovieSceneEvent' (0x49e4042dac0) {}>"]
         [4632, "<Struct 'MovieSceneEvent' (0x49e4042db60) {}>"]
         [4931, "<Struct 'MovieSceneEvent' (0x49e4042dc00) {}>"]
         [5283, "<Struct 'MovieSceneEvent' (0x49e4042dca0) {}>"]
         [5358, "<Struct 'MovieSceneEvent' (0x49e4042dd40) {}>"]
        ]
       }]
     }]
   }
  ]
 }
 Prop_Rocks_Piles_25 {
  id=#63
  possessed_class=Prop_Rocks_Piles_5_C
  tracks [
   {
    type=MovieSceneSpawnTrack
    sections [{type=MovieSceneSpawnSection, channels [None {type=MovieSceneScriptingBoolChannel, default=false, keys [[0,false],[3172,true]]}]}]
   }
   {
    type=MovieScene3DTransformTrack
    sections [{
      type=MovieScene3DTransformSection
      channels [
       Location.X {type=MovieSceneScriptingDoubleChannel, default=108320}
       Location.Y {type=MovieSceneScriptingDoubleChannel, default=-24768}
       Location.Z {type=MovieSceneScriptingDoubleChannel, default=0}
       Rotation.X {type=MovieSceneScriptingDoubleChannel, default=0}
       Rotation.Y {type=MovieSceneScriptingDoubleChannel, default=0}
       Rotation.Z {type=MovieSceneScriptingDoubleChannel, default=-29.3819}
       Scale.X {type=MovieSceneScriptingDoubleChannel, default=1}
       Scale.Y {type=MovieSceneScriptingDoubleChannel, default=1}
       Scale.Z {type=MovieSceneScriptingDoubleChannel, default=1}
      ]
     }]
   }
  ]
 }
]
display_rate {denominator=1, numerator=30}
tick_resolution {denominator=1, numerator=24000}
tracks [
 {type=MovieSceneEventTrack, sections [{type=MovieSceneEventSection}]}
 {
  type=MovieSceneMaterialParameterCollectionTrack
  display_name=SmokeTrailPhase
  sections [{
    type=MovieSceneParameterSection
    channels [
     Phase {
      type=MovieSceneScriptingFloatChannel
      default=0.98
      keys [[1940.8288,0.01],[2019.4125,0.0222],[2065,0.02],[2137,0.033],[2250,0.06],[2445,0.25],[2583.605,0.36],[2632.1687,0.4],[2703.9225,0.475],[2788.4013,0.64],[2850.8263,0.86],[2882,0.98],[3706.9475,2],[4398.1125,3]]
     }
     ZHeight {
      type=MovieSceneScriptingFloatChannel
      default=462437.875
      thinned_from=28
      keys [[1591,7363.7915],[1940.8288,7363.7915],[2019.4125,7800.5278],[2065,7983.4204],[2137,8926.0684],[2250,11958.6465],[2445,20410.457],[2583.605,28362],[2632.1687,31440.8711],[2703.9225,36209.5312],[2788.4013,47550.8633],[2850.8263,59617.4922],[2882,65161.0703],[2912.1513,68065.2344],[2936.94,70005.7188],[2943.3663,71898.0859],[2953.7188,75217.9141],[2968.6588,80582.8906],[3004.12,95840.125],[3061.7125,127157.9062],[3120.7987,165834.5312],[3177.3887,207078.2656],[3285.9462,290557.9688],[3335.7763,327781.125],[3393.615,367729.9375],[3446.0938,399299.25],[3496,462437.875]]
     }
    ]
   }]
 }
 {
  type=MovieSceneMaterialParameterCollectionTrack
  display_name=MPC_Geode
  sections [{
    type=MovieSceneParameterSection
    channels [
     AllowVertCollapseEffect {
      type=MovieSceneScriptingFloatChannel
      default=1
      keys [[0,0],[4298,1],[4325,1],[4381,0],[4628.7,1],[4644,0],[4903,1],[4950,0],[5283,1],[5322,0]]
     }
     Hose_Emissive {
      type=MovieSceneScriptingFloatChannel
      default=20
      keys [[0,0],[1800,5],[1948,20],[4293,20],[4342,200],[4629,1],[4930,200],[5291,5],[5322,200]]
     }
     Blinking Light Speed {type=MovieSceneScriptingFloatChannel, default=0, keys [[0,0],[1200,0.25],[1800,1],[1945,0]]}
     Blinking Light Intensity {type=MovieSceneScriptingFloatChannel, default=0, keys [[0,0],[1200,100]]}
     Entering Or Exiting Tear {type=MovieSceneScriptingFloatChannel, default=1, keys [[4298,0],[4624,1],[4903,0],[5283,1]]}
     Hose Speed {type=MovieSceneScriptingFloatChannel, default=5, keys [[4262.0013,1],[4487.0013,5],[4629,1],[4938.9,5],[5290,1],[5322,5]]}
     Freeze {type=MovieSceneScriptingFloatChannel, default=0, keys [[0,0.75],[900,1],[1944.9,1],[2005,0]]}
     Freeze Exp {type=MovieSceneScriptingFloatChannel, default=2, keys [[0,2],[900,0.25],[1944.9,0.25],[2005,2]]}
     World Location of RiftR {type=MovieSceneScriptingFloatChannel, default=56483.3125, keys [[0,0],[4298,34792.582],[4626,-45000],[4903,12573.3008],[5283,56483.3125]]}
     World Location of RiftG {
      type=MovieSceneScriptingFloatChannel
      default=-45663.6289
      keys [[0,0],[4298,-59167.4102],[4626,46871.957],[4903,-112582.2656],[5283,-45663.6289]]
     }
     World Location of RiftB {
      type=MovieSceneScriptingFloatChannel
      default=5837.6348
      keys [[0,987654324224],[4298,8679.2822],[4626,15500],[4903,14270.249],[5283,5837.6348]]
     }
     World Location of RiftA {type=MovieSceneScriptingFloatChannel}
     Lights_EmissiveMultiplierR {type=MovieSceneScriptingFloatChannel, default=0, keys [[0,4],[1200,0],[1948,0]]}
     Lights_EmissiveMultiplierG {type=MovieSceneScriptingFloatChannel, default=50, keys [[0,4],[1200,10],[1948,50]]}
     Lights_EmissiveMultiplierB {type=MovieSceneScriptingFloatChannel, default=50, keys [[0,4],[1200,10],[1948,50]]}
     Lights_EmissiveMultiplierA {type=MovieSceneScriptingFloatChannel, const [1,1]}
    ]
   }]
 }
 {type=MovieSceneCinematicShotTrack, display_name=Shots, sections [{type=MovieSceneCinematicShotSection, end_frame=5825, start_frame=-46}]}
 {type=MovieSceneSubTrack, display_name=Subsequences, sections [{type=MovieSceneSubSection, end_frame=7721, start_frame=1800}]}
 {
  type=MovieSceneEventTrack
  display_name=BP_Events
  sections [{
    type=MovieSceneEventTriggerSection
    channels [None {
      type=MovieSceneScriptingEventChannel
      keys [[0, "<Struct 'MovieSceneEvent' (0x49e40429f60) {}>"], [3172, "<Struct 'MovieSceneEvent' (0x49e40429ec0) {}>"]]
     }]
   }]
 }
]
