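# Level Sequence that times audio for a Geode event, apparently a test version given its map path, involving a rocket and portal cracks.
# The rocket object carries layered, overlapping sounds, several shaped by volume keys.
# Four crack marker meshes at fixed transforms play paired sounds staggered in order 1, 3, 2, 4, leading into two sounds on the crack-spawn emitter.
# A master audio track adds one mid-sequence sound, alongside an event track showing no keys.
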LevelSequence /Figment_S04_Geode/Athena/Maps/Test/Events/LevelSequence_Geode_Audio
bindings [
 Crack_Marker_Portal_4 {
  id=#1
  possessed_class=FortStaticMeshActor
  tracks [
   {
    type=MovieScene3DTransformTrack
    sections [{
      type=MovieScene3DTransformSection
      channels [
       Location.X {type=MovieSceneScriptingDoubleChannel, default=56164.7578}
       Location.Y {type=MovieSceneScriptingDoubleChannel, default=-45843.6953}
       Location.Z {type=MovieSceneScriptingDoubleChannel, default=5886.0068}
       Rotation.X {type=MovieSceneScriptingDoubleChannel, default=-46.4258}
       Rotation.Y {type=MovieSceneScriptingDoubleChannel, default=66.7515}
       Rotation.Z {type=MovieSceneScriptingDoubleChannel, default=133.0268}
       Scale.X {type=MovieSceneScriptingDoubleChannel, default=1}
       Scale.Y {type=MovieSceneScriptingDoubleChannel, default=1}
       Scale.Z {type=MovieSceneScriptingDoubleChannel, default=1}
      ]
     }]
   }
   {
    type=MovieSceneAudioTrack
    display_name=Audio
    sections [
     {
      type=MovieSceneAudioSection
      end_frame=5414
      start_frame=5148
      channels [
       Volume {type=MovieSceneScriptingFloatChannel, default=1}
       Pitch {type=MovieSceneScriptingFloatChannel, default=1}
      ]
     }
     {
      type=MovieSceneAudioSection
      end_frame=5434
      start_frame=5269
      channels [
       Volume {type=MovieSceneScriptingFloatChannel, default=1}
       Pitch {type=MovieSceneScriptingFloatChannel, default=1}
      ]
     }
    ]
   }
  ]
 }
 Crack_Marker_Portal_2 {
  id=#2
  possessed_class=FortStaticMeshActor
  tracks [
   {
    type=MovieScene3DTransformTrack
    sections [{
      type=MovieScene3DTransformSection
      channels [
       Location.X {type=MovieSceneScriptingDoubleChannel, default=12574.3643}
       Location.Y {type=MovieSceneScriptingDoubleChannel, default=-112581.4375}
       Location.Z {type=MovieSceneScriptingDoubleChannel, default=14269.5352}
       Rotation.X {type=MovieSceneScriptingDoubleChannel, default=178.5852}
       Rotation.Y {type=MovieSceneScriptingDoubleChannel, default=9.9544}
       Rotation.Z {type=MovieSceneScriptingDoubleChannel, default=-130.4404}
       Scale.X {type=MovieSceneScriptingDoubleChannel, default=1}
       Scale.Y {type=MovieSceneScriptingDoubleChannel, default=1}
       Scale.Z {type=MovieSceneScriptingDoubleChannel, default=1}
      ]
     }]
   }
   {
    type=MovieSceneAudioTrack
    display_name=Audio
    sections [
     {
      type=MovieSceneAudioSection
      end_frame=5310
      start_frame=4886
      channels [
       Volume {type=MovieSceneScriptingFloatChannel, default=1}
       Pitch {type=MovieSceneScriptingFloatChannel, default=1}
      ]
     }
     {
      type=MovieSceneAudioSection
      end_frame=5163
      start_frame=4950
      channels [
       Volume {type=MovieSceneScriptingFloatChannel, default=1}
       Pitch {type=MovieSceneScriptingFloatChannel, default=1}
      ]
     }
    ]
   }
  ]
 }
 Crack_Marker_Portal_3 {
  id=#3
  possessed_class=FortStaticMeshActor
  tracks [
   {
    type=MovieScene3DTransformTrack
    sections [{
      type=MovieScene3DTransformSection
      channels [
       Location.X {type=MovieSceneScriptingDoubleChannel, default=-44968.4219}
       Location.Y {type=MovieSceneScriptingDoubleChannel, default=46873.6445}
       Location.Z {type=MovieSceneScriptingDoubleChannel, default=14658.3398}
       Rotation.X {type=MovieSceneScriptingDoubleChannel, default=-171.4562}
       Rotation.Y {type=MovieSceneScriptingDoubleChannel, default=-3.6737}
       Rotation.Z {type=MovieSceneScriptingDoubleChannel, default=-0.5879}
       Scale.X {type=MovieSceneScriptingDoubleChannel, default=1}
       Scale.Y {type=MovieSceneScriptingDoubleChannel, default=1}
       Scale.Z {type=MovieSceneScriptingDoubleChannel, default=1}
      ]
     }]
   }
   {
    type=MovieSceneAudioTrack
    display_name=Audio
    sections [
     {
      type=MovieSceneAudioSection
      end_frame=4819
      start_frame=4545
      channels [
       Volume {type=MovieSceneScriptingFloatChannel, default=1}
       Pitch {type=MovieSceneScriptingFloatChannel, default=1}
      ]
     }
     {
      type=MovieSceneAudioSection
      end_frame=4815
      start_frame=4635
      channels [
       Volume {type=MovieSceneScriptingFloatChannel, default=1}
       Pitch {type=MovieSceneScriptingFloatChannel, default=1}
      ]
     }
    ]
   }
  ]
 }
 Crack_Marker_Portal_1 {
  id=#4
  possessed_class=FortStaticMeshActor
  tracks [
   {
    type=MovieScene3DTransformTrack
    sections [{
      type=MovieScene3DTransformSection
      channels [
       Location.X {type=MovieSceneScriptingDoubleChannel, default=34819.6602}
       Location.Y {type=MovieSceneScriptingDoubleChannel, default=-59070.6836}
       Location.Z {type=MovieSceneScriptingDoubleChannel, default=8666.4092}
       Rotation.X {type=MovieSceneScriptingDoubleChannel, default=-134.6144}
       Rotation.Y {type=MovieSceneScriptingDoubleChannel, default=-80.5835}
       Rotation.Z {type=MovieSceneScriptingDoubleChannel, default=133.7282}
       Scale.X {type=MovieSceneScriptingDoubleChannel, default=1}
       Scale.Y {type=MovieSceneScriptingDoubleChannel, default=1}
       Scale.Z {type=MovieSceneScriptingDoubleChannel, default=1}
      ]
     }]
   }
   {
    type=MovieSceneAudioTrack
    display_name=Audio
    sections [
     {
      type=MovieSceneAudioSection
      end_frame=4718
      start_frame=4294
      channels [
       Volume {type=MovieSceneScriptingFloatChannel, default=1}
       Pitch {type=MovieSceneScriptingFloatChannel, default=1}
      ]
     }
     {
      type=MovieSceneAudioSection
      end_frame=4583
      start_frame=4369
      channels [
       Volume {type=MovieSceneScriptingFloatChannel, default=1}
       Pitch {type=MovieSceneScriptingFloatChannel, default=1}
      ]
     }
    ]
   }
  ]
 }
 Rocket_MainStaticMesh2 {
  id=#5
  display_name=Rocket_MainStaticMesh
  possessed_class=BP_SVLair_GeodeObject2_C
  tracks [{
    type=MovieSceneAudioTrack
    display_name=Audio
    sections [
     {
      type=MovieSceneAudioSection
      end_frame=2629
      start_frame=1828
      channels [
       Volume {type=MovieSceneScriptingFloatChannel, default=1}
       Pitch {type=MovieSceneScriptingFloatChannel, default=1}
      ]
     }
     {
      type=MovieSceneAudioSection
      end_frame=3288
      start_frame=2345
      channels [
       Volume {type=MovieSceneScriptingFloatChannel, default=1, keys [[2345,0],[2740.6025,1],[2954,0.75],[2988,0],[3579.6025,1],[3758.6025,0]]}
       Pitch {type=MovieSceneScriptingFloatChannel, default=1}
      ]
     }
     {
      type=MovieSceneAudioSection
      end_frame=4433
      start_frame=4156
      channels [
       Volume {type=MovieSceneScriptingFloatChannel, default=1, keys [[4366,1],[4384,0]]}
       Pitch {type=MovieSceneScriptingFloatChannel, default=1}
      ]
     }
     {
      type=MovieSceneAudioSection
      end_frame=3591
      start_frame=2988
      channels [
       Volume {type=MovieSceneScriptingFloatChannel, default=1, keys [[3445,1],[3571,0]]}
       Pitch {type=MovieSceneScriptingFloatChannel, default=1}
      ]
     }
     {
      type=MovieSceneAudioSection
      end_frame=4389
      start_frame=4304
      channels [
       Volume {type=MovieSceneScriptingFloatChannel, default=1, keys [[4377,1],[4384,0]]}
       Pitch {type=MovieSceneScriptingFloatChannel, default=1}
      ]
     }
     {
      type=MovieSceneAudioSection
      end_frame=4975
      start_frame=4676
      channels [
       Volume {type=MovieSceneScriptingFloatChannel, default=1, keys [[4676,0],[4682,1],[4964,1],[4975,0]]}
       Pitch {type=MovieSceneScriptingFloatChannel, default=1}
      ]
     }
     {
      type=MovieSceneAudioSection
      end_frame=3197
      start_frame=2975
      channels [
       Volume {type=MovieSceneScriptingFloatChannel, default=1}
       Pitch {type=MovieSceneScriptingFloatChannel, default=1}
      ]
     }
     {
      type=MovieSceneAudioSection
      end_frame=3716
      start_frame=3497
      channels [
       Volume {type=MovieSceneScriptingFloatChannel, default=1}
       Pitch {type=MovieSceneScriptingFloatChannel, default=1}
      ]
     }
     {
      type=MovieSceneAudioSection
      end_frame=4156
      start_frame=3620
      channels [
       Volume {type=MovieSceneScriptingFloatChannel, default=1}
       Pitch {type=MovieSceneScriptingFloatChannel, default=1}
      ]
     }
     {
      type=MovieSceneAudioSection
      end_frame=3014
      start_frame=2931
      channels [
       Volume {type=MovieSceneScriptingFloatChannel, default=1}
       Pitch {type=MovieSceneScriptingFloatChannel, default=1}
      ]
     }
     {
      type=MovieSceneAudioSection
      end_frame=5443
      start_frame=5319
      channels [
       Volume {type=MovieSceneScriptingFloatChannel, default=1}
       Pitch {type=MovieSceneScriptingFloatChannel, default=1}
      ]
     }
    ]
   }]
 }
 P_Geode_Crack_Spawn {
  id=#6
  possessed_class=Emitter
  tracks [{
    type=MovieSceneAudioTrack
    display_name=Audio
    sections [
     {
      type=MovieSceneAudioSection
      end_frame=5751
      start_frame=5406
      channels [
       Volume {type=MovieSceneScriptingFloatChannel, default=1}
       Pitch {type=MovieSceneScriptingFloatChannel, default=1}
      ]
     }
     {
      type=MovieSceneAudioSection
      end_frame=5649
      start_frame=5406
      channels [
       Volume {type=MovieSceneScriptingFloatChannel, default=1}
       Pitch {type=MovieSceneScriptingFloatChannel, default=1}
      ]
     }
    ]
   }]
 }
]
display_rate {denominator=1, numerator=30}
tick_resolution {denominator=1, numerator=24000}
tracks [
 {type=MovieSceneEventTrack, sections [{type=MovieSceneEventSection}]}
 {
  type=MovieSceneAudioTrack
  display_name=Audio
  sections [{
    type=MovieSceneAudioSection
    end_frame=4089
    start_frame=3817
    channels [
     Volume {type=MovieSceneScriptingFloatChannel, default=1}
     Pitch {type=MovieSceneScriptingFloatChannel, default=1}
     AttachActor {type=MovieSceneScriptingActorReferenceChannel}
    ]
   }]
 }
]
