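# Translucent unlit laser beam material with responsive anti-aliasing, under a test events map folder.
# SplineThicken takes a width from pixel depth divided by view width and a Width parameter, scaling beam thickness with camera distance.
# A texture-coordinate mask, peaking mid-way and squared, is multiplied by a power ramp and Opacity for the fade.
# A bright red vector parameter times Multiplier supplies the glow color.
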Material /Figment_S04_Geode/Athena/Maps/Test/Events/M_LaserBeam
graph {
 properties {bEnableResponsiveAA=true, blendMode=BLEND_Translucent, pixelDepthOffsetMode=PDOM_Legacy, shadingModel=MSM_Unlit, usedShadingModels=Unlit}
 nodes [
  n1 = ViewSize()
  n2 = Divide(A=PixelDepth().R, B=Multiply(A=ComponentMask(None=@n1) {r=true}, B=ScalarParameter() {defaultValue=0.25, parameterName=Width}))
  MaterialFunctionCall(WidthBase=@n2, WidthTip=@n2) {materialFunction=&SplineThicken}
  n3 = TextureCoordinate()
  n4 = ComponentMask(None=@n3) {r=true}
  n5 = Constant() {r=0.5}
  n6 = OneMinus(None=Multiply(A=Distance(A=@n4, B=@n5)) {constB=2})
  n7 = ComponentMask(None=@n3) {r=true}
  Multiply(A=Multiply(A=Multiply(A=@n6, B=@n6), B=Power(Base=@n7)), B=ScalarParameter() {defaultValue=1, parameterName=Opacity})
  Multiply(A=VectorParameter() {defaultValue=<rgba 20 0 0 1>}.RGB, B=ScalarParameter() {defaultValue=1, parameterName=Multiplier})
  OneMinus(None=Clamp(None=@n2))
  ScalarParameter() {parameterName=Poop}
  ScalarParameter() {defaultValue=0.1}
 ]
}
